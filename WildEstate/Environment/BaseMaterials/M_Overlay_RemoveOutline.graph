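# Translucent unlit overlay material for skeletal and deformed meshes, named for removing outlines, that allows custom depth writes.
# It samples a packed Shell Fur Map and selects fur length through a channel mask parameter, green by default.
# That length, scaled by an Offset parameter and multiplied by the vertex world normal, forms a shell-style displacement vector.
# Style, clothing and normal masks from the same map go into named reroutes that nothing here reads.
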Material /WildEstate/Environment/BaseMaterials/M_Overlay_RemoveOutline
graph {
 properties {
  allowTranslucentCustomDepthWrites=true
  bAutomaticallySetUsageInEditor=false
  bUsedWithMeshDeformer=true
  bUsedWithSkeletalMesh=true
  blendMode=BLEND_Translucent
  opacityMaskClipValue=0
  shadingModel=MSM_Unlit
  usedShadingModels=Unlit
 }
 nodes [
  Constant()
  n1 = TextureSampleParameter2D() {
   desc="R = Fur Blend Map \r\nG = Length Map "
   group=Fur_Shell_01 - Base Settings
   parameterName=Shell Fur Map
   samplerSource=SSM_Wrap_WorldGroupSettings
   samplerType=SAMPLERTYPE_Masks
   sortPriority=1
   texture=/Game/Global/Textures/Default/Blanks/T_Blank_Mask_Green_R0G1B0A1
  }
  n2 = ChannelMaskParameter(None=@n1.RGBA) {maskChannel=Green, parameterName=FurLegnthChannel, defaultValue=<rgba 0 1 0 0>}
  Multiply(
   A=VertexNormalWS()
   B=Multiply(
    A=NamedRerouteDeclaration(None=@n2.R) {name=FurLengthMap, nodeColor=<rgba 0 1 0.2002 1>}
    B=ScalarParameter() {defaultValue=3.8258, parameterName=Offset}
   )
  )
  NamedRerouteDeclaration(None=ChannelMaskParameter(None=@n1.RGBA) {maskChannel=Blue, parameterName=FurClothingMaskChannel, defaultValue=<rgba 0 0 1 0>}.R) {name=FurClothingMask, nodeColor=<rgba 0.8281 0 1 1>}
  NamedRerouteDeclaration(None=ChannelMaskParameter(None=@n1.RGBA) {parameterName=FurStyleMaskChannel}.R) {name=FurStyleBlendMask, nodeColor=<rgba 1 0 0.0967 1>}
  NamedRerouteDeclaration(None=ChannelMaskParameter(None=@n1.RGBA) {maskChannel=Alpha, parameterName=FurNormalMaskChannel, defaultValue=<rgba 0 0 0 1>}.R) {name=FurNormalMask, nodeColor=<rgba 0.7109 1 0 1>}
 ]
}
