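# Translucent, two-sided unlit material, Nanite-enabled, for thin lines in the WildEstate environment.
# Emissive is base color times a time-of-day tint, with an added color for material proxies (HLOD).
# Opacity takes an optional mask channel and optional pixel-depth fade toward OpacityFar.
# These attributes feed a destructible-damage and cracks function chain and, separately, a SplineThicken node that sets line width through World Position Offset.
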Material /WildEstate/Environment/BaseMaterials/M_ThinLinesNoOutline
graph {
 properties {
  bAutomaticallySetUsageInEditor=false
  bEnableResponsiveAA=true
  bUseMaterialAttributes=true
  bUsedWithNanite=true
  blendMode=BLEND_Translucent
  shadingModel=MSM_Unlit
  twoSided=true
  usedShadingModels=Unlit
 }
 nodes [
  n1 = VectorParameter() {group=Shading, parameterName=Color Base, defaultValue=<rgba 0 0 0 1>}
  n2 = Multiply(
   A=MaterialFunctionCall(
    Color_Morning=VectorParameter() {group=TODM, parameterName=Color Morning, defaultValue=<rgba 0.8906 0.8533 0.7329 1>}.RGB
    Color_Day=VectorParameter() {group=TODM, parameterName=Color Day, defaultValue=<rgba 0.9635 0.9635 0.9635 1>}.RGB
    Color_Evening=VectorParameter() {group=TODM, parameterName=Color Evening, defaultValue=<rgba 0.8229 0.7844 0.66 1>}.RGB
    Color_Night=VectorParameter() {group=TODM, parameterName=Color Night, defaultValue=<rgba 0.4464 0.4969 0.5885 1>}.RGB
   ) {materialFunction=&Color_Time_of_Day}.Result
   B=@n1.RGB
  )
  n3 = VectorParameter() {parameterName=HLOD_ColorAdd, defaultValue=<rgba 0.2344 0.2344 0.2344 1>}
  n4 = ScalarParameter() {defaultValue=2.3, group=Shading, parameterName=OpacityFar}
  n5 = Subtract(A=PixelDepth().R, B=ScalarParameter() {defaultValue=50, group=Shading, parameterName=Opacity_Fade_Offset})
  n6 = ScalarParameter() {defaultValue=50, group=Shading, parameterName=Opacity_Fade_Distance}
  n7 = OneMinus(None=Saturate(None=Divide(A=@n5, B=@n6)))
  n8 = ScalarParameter() {defaultValue=1, group=Shading, parameterName=Opacity}
  n9 = TextureSampleParameter2D() {group=Textures, parameterName=OpacityMask, samplerType=SAMPLERTYPE_Masks, texture=&T_WhiteMask}
  n10 = StaticSwitchParameter(True=Multiply(A=@n8, B=ChannelMaskParameter(None=@n9.RGB) {group=Shading, parameterName=OpacityMaskChannel}.R), False=@n8) {group=Shading, parameterName=Use Opacity mask}
  n11 = StaticSwitchParameter(True=Multiply(A=LinearInterpolate(A=@n4, Alpha=@n7), B=@n10), False=@n10) {group=Shading, parameterName=Use Opacity Fade}
  n12 = SetMaterialAttributes!(
   MaterialAttributes
   Emissive Color=MaterialProxyReplace(Realtime=@n2, MaterialProxy=Add(A=@n2, B=@n3.RGB))
   Opacity=@n11
   Opacity Mask=@n11
   Base Color=@n1.RGB
  ) {attributeSetTypes [
    #1
    #2
    #3
    #4
   ]}
  MaterialFunctionCall!(
   In
   In_Translucent=MaterialFunctionCall(Material_In=@n12, UseCrackedDamageTextures=StaticBool(), UseAdditiveWPO=StaticBool() {value=true}) {materialFunction=&MF_FN_DestructibleDamage}.Material_Out
   DestructibleDamageAmount
  ) {materialFunction=/WildEstate/Environment/BaseMaterials/MaterialFunctions/MF_WildEstate_Cracks}
  n13 = ScalarParameter() {defaultValue=32.5816, parameterName=Width}
  SetMaterialAttributes!(
   MaterialAttributes=@n12
   World Position Offset=MaterialFunctionCall(WidthBase=@n13, WidthTip=@n13, Expand U or V UV Channel=StaticBool()) {materialFunction=&SplineThicken}.WorldPositionOffset
  ) {attributeSetTypes [#5]}
 ]
}
