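# Base anime-style material for WildEstate environment meshes, using world-space normals and supporting Nanite, instanced and spline meshes.
# Shared character anime shader functions (Pre, Calculations, eye animation, Post) build the material attributes.
# Static switches optionally posterize normals using camera-facing angle, with mobile fallbacks, and overlay stepped sun-direction shading on emissive.
# Destructible damage and crack functions follow, then material proxy baking gets a tangent-space normal.
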Material /WildEstate/Environment/BaseMaterials/M_WildEstate_BaseAnime
graph {
 properties {
  bAutomaticallySetUsageInEditor=false
  bTangentSpaceNormal=false
  bUseMaterialAttributes=true
  bUsedWithInstancedStaticMeshes=true
  bUsedWithNanite=true
  bUsedWithSplineMeshes=true
  shadingModel=MSM_FromMaterialExpression
  usedShadingModels=Cloth
 }
 nodes [
  n1 = MaterialFunctionCall!(CustomUV) {materialFunction=/Game/Global/Materials/Character_Parent_2023/Anime_Parent_2023/MaterialFunctions/MFC_AnimeShader_2023_Pre}
  n2 = MaterialFunctionCall!(CustomLightVector, OverrideNormal_WS, STT=@n1.STT, Normal_Map=@n1.STT, DFL=@n1.STT, FX=@n1.STT) {materialFunction=/Game/Global/Materials/Character_Parent_2023/Anime_Parent_2023/MaterialFunctions/MFC_AnimeShader_2023_Calculations}
  n3 = VertexColor()
  n4 = TextureSampleParameter2D() {
   group=01_EyeAnimation_Parameters
   parameterName=EyeMask
   samplerSource=SSM_Wrap_WorldGroupSettings
   samplerType=SAMPLERTYPE_Masks
   sortPriority=-32
   texture=&T_WhiteMask
  }
  n5 = StaticSwitchParameter(True=AppendVector(A=@n3, B=@n3), False=@n4.RGBA) {group=01_EyeAnimation_Parameters, parameterName=UseVertexColorForEyeMask, sortPriority=-33}
  n6 = TextureCoordinate() {coordinateIndex=1}
  n7 = Sign(None=Subtract(A=ComponentMask(None=@n6) {r=true}) {constB=0.0001})
  n8 = MaterialFunctionCall!(
   Color Lit=@n1.Color_Lit
   Color Shaded=@n1.Color_Lit
   Use UV2=StaticBoolParameter() {defaultValue=true, group=01_EyeAnimation_Parameters, parameterName=Use UV2 for Eye Animation, sortPriority=-30}
   Mask=StaticSwitchParameter(
    True=ChannelMaskParameter(None=@n5) {group=01_EyeAnimation_Parameters, parameterName=EyeMaskChannel, sortPriority=-32}.R
    False=Saturate(None=@n7)
   ) {group=01_EyeAnimation_Parameters, parameterName=UseEyeMask, sortPriority=-33}
  ) {materialFunction=/Game/Global/Materials/Character_Parent_2023/Anime_Parent_2023/MaterialFunctions/MF_AnimeEyeAnimation_v2}
  n9 = MaterialFunctionCall!(
   Shading_Gradient=@n2.Shading_Gradient
   Threshold_Gradient=@n2.Shading_Gradient
   Specular_Gradient=@n2.Shading_Gradient
   Sheen_Mask=@n2.Shading_Gradient
   Thickness=@n2.Shading_Gradient
   DFL_Mask=@n2.Shading_Gradient
   FX=@n2.Shading_Gradient
   Metallic_Gradient=@n2.Shading_Gradient
   RimLight_Gradient=@n2.Shading_Gradient
   NormalBuffer=@n2.Shading_Gradient
   --------
   CINE_LightSources=@n2.Shading_Gradient
   CINE_LightColor=@n2.Shading_Gradient
   ---------
   Color_Lit=@n8.Lit
   Color_Shaded=@n8.Lit
   SRM=@n1.SRM
   Emissive=@n1.SRM
   DFL_Color=@n1.SRM
   Specular_Color=@n1.SRM
   RimLight_Color=@n1.SRM
  ) {materialFunction=/Game/Global/Materials/Character_Parent_2023/Anime_Parent_2023/MaterialFunctions/MFC_AnimeShader_2023_Post}
  n10 = GetMaterialAttributes(None=@n9.Result) {attributeGetTypes [#1]}
  n11 = CameraVectorWS()
  n12 = DotProduct(A=@n10.Normal, B=@n11)
  n13 = Abs(None=@n12)
  n14 = ScalarParameter() {defaultValue=3, group=QuantizeOverlay, parameterName=QuantizeNormalSteps}
  n15 = Divide(A=Frac(None=Multiply(A=@n13, B=@n14)), B=@n14)
  n16 = ShadingPathSwitch(Default=LinearInterpolate(A=@n10.Normal, B=@n11, Alpha=Subtract(A=@n12, B=@n15)), Mobile=@n10.Normal)
  n17 = ShadingPathSwitch(Default=@n16, Mobile=VertexNormalWS())
  n18 = StaticSwitchParameter(
   True=SetMaterialAttributes!(
    MaterialAttributes=@n10.MaterialAttributes
    Normal=StaticSwitchParameter(True=@n17, False=@n16) {group=QuantizeOverlay, parameterName=QuantizNormalsDisableOnMobile}
   ) {attributeSetTypes [#1]}
   False=@n9.Result
  ) {group=QuantizeOverlay, parameterName=QuantizNormals}
  n19 = GetMaterialAttributes(None=@n18) {attributeGetTypes [#2, #3]}
  n20 = ScalarParameter() {group=QuantizeOverlay, parameterName=Quantize_MinRangeDarkness, sortPriority=1101}
  n21 = ScalarParameter() {defaultValue=0.75, group=QuantizeOverlay, parameterName=Quantize_MaxRangeBrigthness, sortPriority=1102}
  n22 = CollectionParameter() {
   collection=&FortniteMaterialParameters
   parameterName=SunlightDirectionalVector
  }
  n23 = ComponentMask(None=@n22) {b=true, g=true, r=true}
  n24 = Multiply(A=@n23, B=VectorParameter() {group=QuantizeOverlay, parameterName=QuantizeSunDirection_Mult, defaultValue=<rgba 1 1 1 1>}.RGB)
  n25 = DotProduct(
   A=VertexNormalWS()
   B=StaticSwitchParameter(True=Normalize(VectorInput=@n24), False=@n23) {group=QuantizeOverlay, parameterName=QuantizeOverlay_Modify_SunDirection}
  )
  n26 = ScalarParameter() {defaultValue=0.25, group=QuantizeOverlay, parameterName=Quantize_CenterFlatten, sortPriority=1001}
  n27 = Saturate(None=Subtract(A=Abs(None=@n25), B=@n26))
  n28 = Subtract(A=Constant() {r=1}, B=@n26)
  n29 = Add(A=Multiply(A=Divide(A=@n27, B=@n28), B=Sign(None=@n25)))
  n30 = ScalarParameter() {defaultValue=8, group=QuantizeOverlay, parameterName=Quantize_Steps}
  n31 = Round(None=Multiply(A=Divide(A=@n29), B=@n30))
  n32 = Min(
   A=Max(A=Divide(A=@n31, B=@n30), B=ScalarParameter() {defaultValue=0.25, group=QuantizeOverlay, parameterName=Quantize_MinClamp, sortPriority=1001})
   B=ScalarParameter() {defaultValue=0.75, group=QuantizeOverlay, parameterName=Quantize_MaxClamp, sortPriority=1002}
  )
  n33 = MaterialFunctionCall(Base=@n19.Emissive Color, Blend=Saturate(None=LinearInterpolate(A=@n20, B=@n21, Alpha=@n32))) {materialFunction=&Blend_Overlay}
  NamedRerouteDeclaration(None=StaticSwitchParameter(
    True=SetMaterialAttributes!(MaterialAttributes=@n19.MaterialAttributes, Emissive Color=@n33.Result) {attributeSetTypes [#2]}
    False=@n18
   ) {group=QuantizeOverlay, parameterName=QuantizeOveraly}) {name=AnimeBase, nodeColor=<rgba 0 0.5312 1 1>}
  n34 = NamedRerouteUsage()
  n35 = StaticBool()
  n36 = StaticBool() {value=true}
  n37 = GetMaterialAttributes(None=MaterialFunctionCall!(
    In=MaterialFunctionCall(Material_In=@n34, UseCrackedDamageTextures=@n35, UseAdditiveWPO=@n36) {materialFunction=&MF_FN_DestructibleDamage}.Material_Out
    In_Translucent
    DestructibleDamageAmount
   ) {materialFunction=/WildEstate/Environment/BaseMaterials/MaterialFunctions/MF_WildEstate_Cracks}.Result) {attributeGetTypes [#1]}
  SetMaterialAttributes!(
   MaterialAttributes=@n37.MaterialAttributes
   Normal=MaterialProxyReplace(Realtime=@n37.Normal, MaterialProxy=Transform(None=@n37.Normal) {transformSourceType=TRANSFORMSOURCE_World, transformType=TRANSFORM_Tangent})
  ) {attributeSetTypes [#1]}
 ]
}
