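# Material function that builds the top-blend layer material for Wild Estate cliffs, tinted per landscape biome.
# Biome masks choose between Default, A, C and D color pairs, each varied by a mask.
# A Z-projected world-aligned mask texture multiplies that color, crossfading near to far outlines by pixel depth.
# Output material attributes set constant specular, switched constant roughness and an optional flattened, transformed world-aligned normal.
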MaterialFunction /WildEstate/Environment/BaseMaterials/MaterialFunctions/MF_Cliff_TopBlendMat
graph {nodes [
  Constant3Vector() {constant=<rgba 0 0 1 0>}
  n1 = VectorParameter() {group=Top Blend Materail, parameterName=TopBlendMat_Default_2, sortPriority=80, defaultValue=<rgba 0.1146 0.1006 0.0868 1>}
  n2 = VectorParameter() {group=Top Blend Materail, parameterName=TopBlendMat_Default_1, sortPriority=80, defaultValue=<rgba 0.1458 0.1281 0.1105 1>}
  n3 = NamedRerouteUsage()
  n4 = VectorParameter() {group=Top Blend Materail, parameterName=TopBlendMat_BIomeA_2, sortPriority=80, defaultValue=<rgba 0.125 0.1137 0.0786 1>}
  n5 = VectorParameter() {group=Top Blend Materail, parameterName=TopBlendMat_BIomeA_1, sortPriority=80, defaultValue=<rgba 0.1042 0.0903 0.0678 1>}
  n6 = NamedRerouteUsage()
  n7 = MaterialFunctionCall!(UseLandscapeCoords=StaticBool(), "UseProjection(translucent)") {materialFunction=/WildEstate/Landscape/MaterialFunctions/MF_WildEstate_BiomeMasks}
  n8 = MaterialFunctionCall(Float3=@n7.QuadA_RGBA) {materialFunction=&BreakOutFloat3Components}
  n9 = VectorParameter() {group=Top Blend Materail, parameterName=TopBlendMat_BIomeC_2, sortPriority=80, defaultValue=<rgba 0.2865 0.1623 0.0998 1>}
  n10 = VectorParameter() {group=Top Blend Materail, parameterName=TopBlendMat_BIomeC_1, sortPriority=80, defaultValue=<rgba 0.2083 0.1181 0.0726 1>}
  n11 = LinearInterpolate(
   A=LinearInterpolate(A=LinearInterpolate(A=@n1.RGB, B=@n2.RGB, Alpha=@n3), B=LinearInterpolate(A=@n4.RGB, B=@n5.RGB, Alpha=@n6), Alpha=@n8.R)
   B=LinearInterpolate(A=@n9.RGB, B=@n10.RGB, Alpha=NamedRerouteUsage())
   Alpha=@n8.G
  )
  n12 = VectorParameter() {group=Top Blend Materail, parameterName=TopBlendMat_BIomeD_2, sortPriority=80, defaultValue=<rgba 0.3958 0.1532 0.1022 1>}
  n13 = VectorParameter() {group=Top Blend Materail, parameterName=TopBlendMat_BIomeD_1, sortPriority=80, defaultValue=<rgba 0.3594 0.1366 0.0893 1>}
  n14 = LinearInterpolate(A=@n12.RGB, B=@n13.RGB, Alpha=NamedRerouteUsage())
  n15 = MaterialFunctionCall(Float3=@n7.QuadB_RGBA) {materialFunction=&BreakOutFloat3Components}
  n16 = TextureObjectParameter() {
   group=Top Blend Materail
   parameterName=TopBlendMat_Mask_Texture
   samplerType=SAMPLERTYPE_Masks
   sortPriority=40
   texture=/WildEstate/Environment/Terrain/Cliffs/Textures/T_WildeEstate_Cliff_Gorge_A_D_Paul
   channelNames {a="", b=ColorVariationMask, g=OutlineFar, r=OutlineNear}
  }
  n17 = ScalarParameter() {defaultValue=1200, group=Top Blend Materail, parameterName=TopBlendMat__Overlay_Texture_Size, sortPriority=20}
  n18 = ScalarParameter() {defaultValue=200, group=Top Blend Materail, parameterName=TopBlendMat_Overlay_Texture_ProjcectionContrast, sortPriority=20}
  n19 = MaterialFunctionCall(TextureObject=@n16, TextureSize=@n17, ProjectionTransitionContrast=@n18) {materialFunction=&WorldAlignedTexture}
  n20 = ComponentMask(None=@n19.Z Texture) {r=true}
  n21 = ComponentMask(None=@n19.Z Texture) {g=true}
  n22 = PixelDepth()
  n23 = ScalarParameter() {defaultValue=5000, group=Top Blend Materail, parameterName=TopBlendMat_OutlineSwitch_OffsetDistance, sortPriority=3240}
  n24 = Saturate(None=Divide(
    A=Subtract(A=@n22.R, B=@n23)
    B=ScalarParameter() {defaultValue=5000, group=Top Blend Materail, parameterName=TopBlendMat_OutlineSwitch_BlendDistance, sortPriority=3240}
   ))
  n25 = SetMaterialAttributes!(
   MaterialAttributes
   BaseColor=Multiply(A=LinearInterpolate(A=@n11, B=@n14, Alpha=@n15.R), B=LinearInterpolate(A=@n20, B=@n21, Alpha=@n24))
   Roughness=StaticSwitchParameter(True=Constant() {desc=CellShadingValue, r=0.88}, False=Constant() {desc=CellShadingValue, r=0.9}) {defaultValue=true, group=Top Blend Materail, parameterName=Use TopLayer Outline, sortPriority=5}
   Specular=Constant()
  ) {attributeSetTypes [#1, #2, #3]}
  n26 = TextureObjectParameter() {
   group=Top Blend Materail
   parameterName=TopBlendMat_Normals
   samplerType=SAMPLERTYPE_Normal
   sortPriority=50
   texture=/WildEstate/Environment/Terrain/Cliffs/Textures/T_WildeEstate_Cliff_Gorge_A_N
  }
  n27 = StaticBool()
  n28 = ScalarParameter() {defaultValue=2, group=Top Blend Materail, parameterName=TopBlendMat_Normal_Texture_ProjcectionContrast, sortPriority=50}
  n29 = Transform(None=MaterialFunctionCall(
    Normal=MaterialFunctionCall(TextureObject=@n26, TextureSize=@n17, WorldSpace=@n27, ProjectionTransitionContrast=@n28) {materialFunction=&WorldAlignedNormal}.XY Texture
    Flatness=ScalarParameter() {defaultValue=0.5, group=Top Blend Materail, parameterName=TopBlendMat_Normal_Flatten, sortPriority=50}
   ) {materialFunction=&FlattenNormal}.Result)
  FunctionOutput(None=StaticSwitchParameter(True=SetMaterialAttributes!(MaterialAttributes=@n25, Normal=@n29) {attributeSetTypes [#4]}, False=@n25) {defaultValue=true, group=Top Blend Materail, parameterName=Use Normal Texture TopBlend, sortPriority=50})
  NamedRerouteDeclaration(None=ComponentMask(None=@n19.Z Texture) {b=true}) {name=TextureColorVarMask, nodeColor=<rgba 0 0.8594 1 1>}
  StaticSwitchParameter(True=@n19.XYZ Texture, False=@n19.XYZ Texture) {group=Top Blend Materail, parameterName="TopBlendMat Z or XYZ projcetion(Trui is XYZ)"}
 ]}
