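# Material function that adds an optional glowing "Nucular" emissive to field materials in WildEstate, toggled by the static switch UseNucularGlowFields.
# World-space crack textures minus panning noise combine with UV box masks for dirt rows and a base-color-style mask.
# Sine pulses, offset by object position, and two switchable color palettes are defined.
# Camera-distance fades blend toward a flat far color, also used for proxies, scaled by time of day and inverse eye adaptation.
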MaterialFunction /WildEstate/Environment/BaseMaterials/MaterialFunctions/MF_NucularGlowFields
graph {nodes [
  n1 = FunctionInput() {inputType=FunctionInput_MaterialAttributes}
  n2 = GetMaterialAttributes(None=@n1) {attributeGetTypes [#1, #2]}
  n3 = Add(A=@n2.EmissiveColor, B=NamedRerouteUsage()) {constB=0}
  FunctionOutput(None=StaticSwitchParameter(
    True=SetMaterialAttributes!(MaterialAttributes=@n2.MaterialAttributes, EmissiveColor=@n3, BaseColor) {attributeSetTypes [#1, #2]}
    False=@n1
   ) {group=NucularGlow, parameterName=UseNucularGlowFields})
  n4 = VectorParameter() {group=NucularGlow, parameterName=Brealth options, channelNames {a="", b=Speed, g=High, r=low}, defaultValue=<rgba 0.9 1 5 0>}
  n5 = Divide(B=@n4.Speed)
  n6 = Time()
  n7 = Sine(None=Multiply(A=@n5, B=@n6))
  NamedRerouteDeclaration(None=LinearInterpolate(A=@n4.low, B=@n4.low, Alpha=ConstantBiasScale(None=@n7))) {name=pulsing, nodeColor=<rgba 0 0.6484 1 1>}
  NamedRerouteDeclaration(None=@n2.BaseColor) {name=Base color, nodeColor=<rgba 1 0.6797 0 1>}
  n8 = ScalarParameter() {desc=Switch to secondary color palette, group=NucularGlow, parameterName=Color2}
  NamedRerouteDeclaration(None=LinearInterpolate(
    A=VectorParameter() {group=NucularGlow, parameterName=NucularGlowColor, defaultValue=<rgba 0.0752 1 0.1728 1>}.RGB
    B=VectorParameter() {group=NucularGlow, parameterName=NucularGlowColor2, defaultValue=<rgba 0.5101 0.1455 1 1>}.RGB
    Alpha=@n8
   )) {name=Nucular_Glow, nodeColor=<rgba 0.3594 0 1 1>}
  n9 = ObjectPositionWS()
  n10 = ScalarParameter() {defaultValue=1, parameterName=Random_Scale}
  n11 = Divide(A=ComponentMask(None=@n9) {r=true}, B=@n10)
  n12 = Divide(A=ComponentMask(None=@n9) {g=true}, B=@n10)
  n13 = Multiply(A=@n5, B=Add(A=Add(A=@n11, B=@n12), B=@n6))
  NamedRerouteDeclaration(None=ConstantBiasScale(None=Sine(None=@n13))) {name=pulse_time_random, nodeColor=<rgba 0.7812 1 0 1>}
  n14 = Multiply(A=NamedRerouteUsage(), B=ScalarParameter() {defaultValue=25.7092, group=NucularGlow, parameterName=BaseColor_Mask_Multi}) {constB=10}
  n15 = ScalarParameter() {defaultValue=0.5, group=NucularGlow, parameterName=BaseColor_Mask_Pow}
  n16 = LinearInterpolate(A=LinearInterpolate(Alpha=NamedRerouteUsage()) {constA=0.8, constB=1.4, desc=was .4 and 1}, Alpha=Saturate(None=Power(Base=@n14, Exp=@n15))) {constA=0.5, constB=0}
  n17 = WorldPosition()
  n18 = ScalarParameter() {defaultValue=2000, group=NucularGlow, parameterName=Nucular_Glow_Cracks_Secondary_Noise}
  n19 = Multiply(
   A=TextureSample(UVs=Divide(A=@n17.XY, B=@n18)) {samplerType=SAMPLERTYPE_LinearColor, texture=/Game/Global/Materials/MaterialGym/Materials/GhostNoise/Textures/T_ghost_swirls_01}.RGB
   B=ScalarParameter() {defaultValue=2, group=NucularGlow, parameterName=secondary_cracks_Mulit}
  )
  n20 = TextureSample(UVs=Divide(A=WorldPosition().XY, B=ScalarParameter() {defaultValue=100, group=NucularGlow, parameterName=Nucular_Glow_Cracks_Far})) {texture=/WildEstate/Environment/Textures/T_WildEstate_Cracks}
  n21 = WorldPosition()
  n22 = ScalarParameter() {defaultValue=3, group=NucularGlow, parameterName=Nucular_Glow_Noise_Scale_Close}
  n23 = TextureObject() {texture=&AtmosphericCloudNoise01}
  n24 = TextureSample(UVs=Panner(Coordinate=Divide(A=@n21.XY, B=@n22)) {speedX=0.02, speedY=0.01}, Tex=@n23) {samplerType=SAMPLERTYPE_Grayscale, texture=/Game/Global/Textures/Default/Noise/TilingNoise06}
  n25 = ScalarParameter() {defaultValue=1, group=NucularGlow, parameterName=Nucular_Glow_Noise_Str_close}
  n26 = Panner(Coordinate=Divide(A=@n21.XY, B=ScalarParameter() {defaultValue=3, group=NucularGlow, parameterName=Nucular_Glow_Noise_Scale_Far})) {speedX=0.04, speedY=0.1}
  n27 = LinearInterpolate(
   A=Power(Base=Multiply(A=@n24.RGB, B=@n25))
   B=Multiply(
    A=TextureSample(UVs=@n26, Tex=@n23) {samplerType=SAMPLERTYPE_Grayscale, texture=&TilingNoise06}.RGB
    B=ScalarParameter() {group=NucularGlow, parameterName=Nucular_Glow_Noise_Str_Far}
   )
   Alpha=NamedRerouteUsage()
  )
  n28 = Subtract(A=LinearInterpolate(A=Power(Base=@n19), B=@n20.RGB, Alpha=NamedRerouteUsage()), B=@n27)
  n29 = VectorParameter() {
   group=NucularGlow
   parameterName=Box_mask
   channelNames {a=BoundsY, b=BoundsX, g=LocationY, r=LocationX}
   defaultValue=<rgba 0.5 0.5 0.1 0.7>
  }
  n30 = MaterialFunctionCall!(
   A=TextureCoordinate()
   B=ComponentMask(None=@n29.RGBA) {g=true, r=true}
   Bounds=ComponentMask(None=@n29.RGBA) {a=true, b=true}
   Edge Falloff=ScalarParameter() {group=NucularGlow, parameterName=Box_Falloff}
  ) {materialFunction=/Engine/Functions/Engine_MaterialFunctions01/Gradient/BoxMask-2D}
  n31 = Multiply(
   A=Multiply(A=Saturate(None=@n28), B=@n30.Result)
   B=LinearInterpolate(
    A=ScalarParameter() {defaultValue=0.3, group=NucularGlow, parameterName=Nucular_Noise_Str_Close}
    B=ScalarParameter() {group=NucularGlow, parameterName=Nucular_Noise_Str_Far}
    Alpha=NamedRerouteUsage()
   ) {constA=0.3}
  )
  n32 = NamedRerouteUsage()
  n33 = NamedRerouteUsage()
  n34 = Multiply(A=Saturate(None=Subtract(A=@n20.RGB, B=@n27)), B=@n30.Result)
  n35 = LinearInterpolate(Alpha=NamedRerouteUsage()) {constA=0.1, constB=0.3}
  n36 = VectorParameter() {
   group=NucularGlow
   parameterName=Dirt_Mask_mask
   channelNames {a=BoundsY, b=BoundsX, g=LocationY, r=LocationX}
   defaultValue=<rgba 0.5 0.5 0.1 0.7>
  }
  n37 = ComponentMask(None=@n36.RGBA) {g=true, r=true}
  n38 = ComponentMask(None=@n36.RGBA) {a=true, b=true}
  n39 = LinearInterpolate(
   A=ScalarParameter() {group=NucularGlow, parameterName=Dirt_Falloff}
   B=ScalarParameter() {defaultValue=0.48, group=NucularGlow, parameterName=Dirt_Falloff_Far}
   Alpha=NamedRerouteUsage()
  )
  n40 = MaterialFunctionCall(A=TextureCoordinate(), B=@n37, Bounds=@n38, Edge Falloff=@n39) {materialFunction=&BoxMask-2D}
  n41 = AppendVector(A=ScalarParameter() {group=NucularGlow, parameterName=Offset_Row}, B=Constant())
  n42 = MaterialFunctionCall(A=TextureCoordinate(), B=Add(A=@n37, B=@n41), Bounds=@n38, Edge Falloff=@n39) {materialFunction=&BoxMask-2D}
  n43 = TextureCoordinate()
  n44 = Subtract(A=@n37, B=@n41)
  n45 = Saturate(None=Add(
    A=Add(A=@n40.Result, B=@n42.Result)
    B=MaterialFunctionCall(A=@n43, B=@n44, Bounds=@n38, Edge Falloff=@n39) {materialFunction=&BoxMask-2D}.Result
   ))
  n46 = LinearInterpolate(A=@n45, B=Multiply(A=ScalarParameter() {group=NucularGlow, parameterName=glow_dirt_Distance_medium}, B=@n45), Alpha=NamedRerouteUsage())
  n47 = Power(Base=Multiply(A=@n46, B=ScalarParameter() {defaultValue=3, group=NucularGlow, parameterName=glow_dirt_Mult}))
  n48 = Saturate(None=Add(A=Multiply(A=@n34, B=@n35) {constB=0.2}, B=@n47) {constB=0})
  n49 = Multiply(A=@n47, B=ScalarParameter() {defaultValue=3, group=NucularGlow, parameterName=glow_dirt_Mult_far})
  n50 = Saturate(None=Add(A=Multiply(A=@n31, B=@n32), B=LinearInterpolate(A=@n48, B=@n49, Alpha=@n33) {constA=0.3, constB=0}))
  n51 = Multiply(
   A=NamedRerouteUsage()
   B=Multiply(A=Add(A=@n16, B=@n50), B=ScalarParameter() {defaultValue=0.7, group=NucularGlow, parameterName=global_Multi_Nucular}) {constB=0.7}
  )
  n52 = NamedRerouteUsage()
  n53 = LinearInterpolate(
   A=VectorParameter() {group=NucularGlow, parameterName=NucularGlowColorFar, defaultValue=<rgba 0.2082 1 0.0437 1>}.RGB
   B=VectorParameter() {group=NucularGlow, parameterName=NucularGlowColorFar2, defaultValue=<rgba 0.3878 0.0812 0.8438 1>}.RGB
   Alpha=@n8
  )
  n54 = ObjectPositionWS()
  n55 = CameraPositionWS()
  n56 = Divide(
   A=Subtract(A=Distance(A=@n54, B=@n55), B=ScalarParameter() {defaultValue=21457.4141, group=NucularGlow, parameterName=CameraFadeOffset})
   B=ScalarParameter() {defaultValue=170616.9844, group=NucularGlow, parameterName=CameraFadeLength}
  )
  n57 = MaterialProxyReplace(Realtime=LinearInterpolate(A=Multiply(A=@n51, B=@n52), B=@n53, Alpha=Clamp(None=@n56)), MaterialProxy=@n53)
  n58 = MaterialFunctionCall(
   Value_Morning=ScalarParameter() {defaultValue=0.6, group=NucularGlow, parameterName="Value_Morning (S)"}
   Value_Day=ScalarParameter() {defaultValue=0.65, group=NucularGlow, parameterName="Value_Day (S)"}
   Value_Evening=ScalarParameter() {defaultValue=0.65, group=NucularGlow, parameterName="Value_Evening (S)"}
   Value_Night=ScalarParameter() {defaultValue=0.5, group=NucularGlow, parameterName="Value_Night (S)"}
  ) {materialFunction=&Scalar_Time_of_Day}
  NamedRerouteDeclaration(None=EyeAdaptationInverse(LightValueInput=Multiply(A=@n57, B=@n58.Result)).EyeAdaptationInverse) {name=emissive output, nodeColor=<rgba 1 0.5859 0 1>}
  n59 = WorldPosition()
  n60 = Subtract(
   A=Distance(A=@n59.XYZ, B=CameraPositionWS())
   B=ScalarParameter() {defaultValue=21457.4141, group=NucularGlow, parameterName=CameraFadeOffset_medium}
  )
  n61 = ScalarParameter() {defaultValue=170616.9844, group=NucularGlow, parameterName=CameraFadeLength_Medium}
  NamedRerouteDeclaration(None=Clamp(None=Divide(A=@n60, B=@n61))) {name=medium_distance_Fade, nodeColor=<rgba 0 1 0.8125 1>}
  n62 = Multiply(A=@n46, B=ScalarParameter() {defaultValue=10, group=NucularGlow, parameterName=glow_Dirt_Mask_Str}) {constB=10}
  NamedRerouteDeclaration(None=LinearInterpolate(A=Power(Base=@n62), Alpha=NamedRerouteUsage()) {constA=0.3}) {name=Dirt Box Mask Medium, nodeColor=<rgba 0 0.8125 1 1>}
  n63 = Subtract(
   A=Distance(A=@n59.XYZ, B=CameraPositionWS())
   B=ScalarParameter() {defaultValue=21457.4141, group=NucularGlow, parameterName=CameraFadeOffset_CornRow}
  )
  n64 = ScalarParameter() {defaultValue=170616.9844, group=NucularGlow, parameterName=CameraFadeLength_CornRow}
  NamedRerouteDeclaration(None=Clamp(None=Divide(A=@n63, B=@n64))) {name=Cornrow_Fade, nodeColor=<rgba 0 1 0.8125 1>}
 ]}
