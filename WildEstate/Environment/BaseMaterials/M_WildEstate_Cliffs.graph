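# Base material for WildEstate cliff meshes, supporting Nanite, instancing and splines, with world-space normals.
# Cliff function output gets optional top layer and vertex-color roughness outlines, then color quantization and flat shading.
# Landscape height and color virtual textures drive grass and ground blending, with noise-broken vertex-color grass and optional normal-based snow on top.
# Glow field, destructible damage and crack functions finish the chain; MaterialProxyReplace converts normals to tangent space for proxy bakes.
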Material /WildEstate/Environment/BaseMaterials/M_WildEstate_Cliffs
graph {
 properties {
  bAutomaticallySetUsageInEditor=false
  bTangentSpaceNormal=false
  bUseMaterialAttributes=true
  bUsedWithInstancedStaticMeshes=true
  bUsedWithNanite=true
  bUsedWithSplineMeshes=true
  physMaterial=/Game/FortressPhysicalMaterials/Grass
  shadingModel=MSM_FromMaterialExpression
  usedShadingModels=Default Lit
 }
 nodes [
  n1 = TwoSidedSign()
  AppendVector(A=AppendVector(A=@n1, B=@n1), B=Constant() {r=1})
  n2 = RuntimeVirtualTextureSampleParameter(World Position=WorldPosition().XYZ) {
   bEnableFeedback=false
   materialType=WorldHeight
   parameterName=RVT_WorldHeight
   virtualTexture=/Game/Athena/Apollo/Maps/Landscape/VT/RVT_Apollo_Landscape_Height
  }
  n3 = Subtract(A=WorldPosition().Z, B=NamedRerouteDeclaration(None=@n2.WorldHeight) {name=RVT_WordPosZ, nodeColor=<rgba 0 0.7656 1 1>})
  n4 = Saturate(None=Divide(
    A=NamedRerouteDeclaration(None=@n3) {name=RVT_WordPosZbiased, nodeColor=<rgba 0 0.7656 1 1>}
    B=ScalarParameter() {defaultValue=16, parameterName=BlendDistance}
   ))
  BlendMaterialAttributes(
   A=SetMaterialAttributes!(MaterialAttributes, Emissive Color, Base Color=Constant3Vector() {constant=<rgba 1 0 0.0213 1>}) {attributeSetTypes [#1, #2]}
   Alpha=@n4
  )
  ComponentMask(None=VertexNormalWS()) {b=true}
  Constant2Vector() {g=1, r=1}
  Constant() {r=0.02}
  n5 = Add(
   A=OneMinus(None=MaterialFunctionCall!(NormalOverrideWS) {materialFunction=/WildEstate/Environment/BaseMaterials/MaterialFunctions/MF_WildEstateWorldNoise}.TransitionNoise)
   B=ScalarParameter() {defaultValue=-1, parameterName=GrassEdgeBias}
  )
  n6 = Subtract(A=PixelDepth().R, B=ScalarParameter() {defaultValue=1000, parameterName=FadeGrassEdgeDetail_Offset})
  n7 = ScalarParameter() {defaultValue=1000, parameterName=FadeGrassEdgeDetail_Fade}
  LinearInterpolate(A=@n5, B=ScalarParameter() {parameterName=GrassEdgeFar}, Alpha=Saturate(None=Divide(A=@n6, B=@n7)))
  n8 = MaterialFunctionCall!(In Material, NormalOverrideWS=NamedRerouteUsage(), LayerWeight, ArrayTextureUV) {materialFunction=/WildEstate/Landscape/M_WildEstate_Snow}
  LinearInterpolate(A=NamedRerouteUsage(), B=@n8.BaseColor, Alpha=NamedRerouteUsage())
  MaterialFunctionCall!(NoRVT, RVT) {materialFunction=/Game/Packages/Fortress_SharedMaterials/Base_Material_Functions/FN_RVT_Switch_Color}
  n9 = NamedRerouteUsage()
  n10 = StaticBool() {value=true}
  n11 = StaticBool()
  n12 = Transform(None=Constant3Vector() {constant=<rgba 0 0 1 0>})
  n13 = NamedRerouteUsage()
  n14 = NamedRerouteUsage()
  n15 = SetMaterialAttributes!(
   MaterialAttributes
   Base Color=Multiply(
    A=MaterialFunctionCall!(In Material, ArrayTextureUV, CliffMask) {materialFunction=/WildEstate/Environment/BaseMaterials/MaterialFunctions/MF_WildEstate_LandscapeDetailTexture}.GrassDetailTextureForStaticMesh
    B=MaterialFunctionCall!(NormalOverrideWS=@n9, UseCamera facing BildBoards=@n10, UseLandscapeCoords=@n11) {materialFunction=/WildEstate/Environment/BaseMaterials/MaterialFunctions/MF_WildEstate_Grass}.Grass BaseColor
   )
   Roughness=MaterialFunctionCall() {materialFunction=/WildEstate/Foliage/BaseMaterial/MF_WildEstate_GrassRoughness}.Result
   Specular=Constant()
   Normal=MaterialFunctionCall(NoRVT=@n12, RVT=LinearInterpolate(A=@n13, B=@n12, Alpha=@n14)) {materialFunction=&FN_RVT_Switch_Color}.Result
  ) {attributeSetTypes [
    #2
    #3
    #4
    #5
   ]}
  n16 = ScalarParameter() {defaultValue=-5, parameterName=GrassTransitionContrast_FB_Low}
  n17 = ScalarParameter() {defaultValue=5, parameterName=GrassTransitionContrast_FB_high}
  n18 = OneMinus(None=VertexColor().R)
  n19 = NaniteReplace!(Default=Saturate(None=LinearInterpolate(A=@n16, B=@n17, Alpha=@n18)), Nanite=@n18)
  n20 = ShadingPathSwitch(Default=Add(A=@n5, B=@n19), Mobile=@n19) {desc=Migrated from FeatureLevelSwitch}
  n21 = Round(None=@n20)
  n22 = MaterialFunctionCall!(In Material, UseLandscapeCoords=StaticBool()) {materialFunction=/WildEstate/Environment/BaseMaterials/MaterialFunctions/MF_WildEstate_Cliffs}
  n23 = VertexNormalWS()
  n24 = ChannelMaskParameter(None=VertexColor()) {group=Top Blend Materail, parameterName=TopLayer_VertexColorChannel, sortPriority=2}
  n25 = StaticSwitchParameter(
   True=OneMinus(None=ComponentMask(None=@n23) {b=true})
   False=StaticSwitchParameter(True=OneMinus(None=@n24.R), False=@n24.R) {group=Top Blend Materail, parameterName=TopLayer_VertexColor_invert, sortPriority=2}
  ) {
   desc="False = VertexColor and True = Up Facing Normals"
   group=Top Blend Materail
   parameterName=TopLayer VertexColor || UpFacingNormal
   sortPriority=2
  }
  n26 = Round(None=Add(A=@n25, B=ScalarParameter() {group=Top Blend Materail, parameterName=TopBlendMat_Transition_Bias}))
  n27 = StaticSwitchParameter(
   True=BlendMaterialAttributes(
    A=MaterialFunctionCall() {materialFunction=/WildEstate/Environment/BaseMaterials/MaterialFunctions/MF_Cliff_TopBlendMat}.Result
    B=@n22.Out For StaticMesh
    Alpha=Saturate(None=@n26)
   )
   False=@n22.Out For StaticMesh
  ) {group=Top Blend Materail, parameterName=Use TopLayer Materail, sortPriority=1}
  n28 = GetMaterialAttributes(None=@n27) {attributeGetTypes [#3]}
  n29 = ChannelMaskParameter(None=VertexColor()) {
   group=Addition VertexColor Outlines
   maskChannel=Blue
   parameterName=VertexColor Outline Channel
   sortPriority=2
   defaultValue=<rgba 0 0 1 0>
  }
  n30 = Add(A=@n29.R, B=ScalarParameter() {group=Addition VertexColor Outlines, parameterName=Additional Vertex Color Bias})
  n31 = SetMaterialAttributes!(MaterialAttributes=@n28.MaterialAttributes, Roughness=LinearInterpolate(A=Constant() {r=0.67}, B=@n28.Roughness, Alpha=Round(None=@n30))) {attributeSetTypes [#3]}
  n32 = MaterialFunctionCall!(
   Material In=MaterialFunctionCall!(In=StaticSwitchParameter(True=@n31, False=@n27) {group=Addition VertexColor Outlines, parameterName=Use Additional Vertex Color Outlines}) {materialFunction=/WildEstate/Environment/BaseMaterials/MaterialFunctions/M_WildEstate_QantizeColor}.Result
   UseFlatShading=StaticBoolParameter() {group=FlatShading, parameterName=UseFlatShading_All, sortPriority=1}
   IsWorldSpace=StaticBool() {value=true}
   OverrideUnlitShadinBaseColorDarkening=ScalarParameter() {defaultValue=0.6, group=FlatShading, parameterName=Cliffs_UnlitFlatShading_BaseColorBrightness}
   OverrideUnlitShadinBaseColorEmissive=ScalarParameter() {defaultValue=0.5, group=FlatShading, parameterName=Cliffs_UnlitFlatShading_Emissive}
  ) {materialFunction=/WildEstate/Environment/BaseMaterials/MaterialFunctions/MF_WildEstate_FlatShading}
  n33 = GetMaterialAttributes(None=@n32.Result) {attributeGetTypes [#5]}
  n34 = Multiply(A=@n33.Normal, B=@n1)
  n35 = MaterialFunctionCall!(
   In=StaticSwitchParameter(
    True=SetMaterialAttributes!(MaterialAttributes=@n33.MaterialAttributes, Normal=@n34) {attributeSetTypes [#5]}
    False=@n32.Result
   ) {group=Shading, parameterName=FlipBackFaceNormals}
   RVT_Normal=NamedRerouteUsage()
   IsWorldSpace=StaticBool() {value=true}
   RVT_WorldPosZ=NamedRerouteUsage()
   RVT_BaseColor=NamedRerouteUsage()
   RVT_Specular=NamedRerouteUsage()
   RVT_Roughness=NamedRerouteUsage()
   L2_Mask
   L3_Mask
   L4_Mask
   IsTree
   UseLayerMasking=StaticBool()
  ) {materialFunction=/WildEstate/Environment/BaseMaterials/MaterialFunctions/M_WildEstate_GrassBlend}
  n36 = StaticSwitchParameter(True=BlendMaterialAttributes(A=@n35.Result, B=@n15, Alpha=Saturate(None=@n21)), False=@n35.Result) {defaultValue=true, group=Grass Blend, parameterName=UseTopLayerGrassBlend}
  n37 = Transform(None=Constant3Vector() {constant=<rgba 0 0 1 0>})
  n38 = NamedRerouteUsage()
  n39 = NamedRerouteUsage()
  n40 = SetMaterialAttributes!(
   MaterialAttributes
   Base Color=@n8.BaseColor
   Roughness=Constant() {r=0.3}
   Specular=Constant()
   Normal=MaterialFunctionCall(NoRVT=@n37, RVT=LinearInterpolate(A=@n38, B=@n37, Alpha=@n39)) {materialFunction=&FN_RVT_Switch_Color}.Result
  ) {attributeSetTypes [
    #2
    #3
    #4
    #5
   ]}
  n41 = VertexNormalWS()
  n42 = Saturate(None=Round(None=ComponentMask(None=@n41) {b=true}))
  n43 = StaticSwitchParameter(True=BlendMaterialAttributes(A=@n36, B=@n40, Alpha=@n42), False=@n36) {group=Grass Blend, parameterName=UseTopLayerSnowBlend}
  n44 = GetMaterialAttributes(None=@n43) {attributeGetTypes [#2, #5]}
  n45 = NamedRerouteUsage()
  n46 = ScalarParameter() {defaultValue=2, group=RVT Blend, parameterName=RVT_Smooth_BlendBias}
  n47 = Saturate(None=Divide(A=Subtract(A=@n45, B=@n46), B=ScalarParameter() {defaultValue=4, group=RVT Blend, parameterName=RVT_Smooth_BlendDistance}))
  n48 = SetMaterialAttributes!(
   MaterialAttributes=@n44.MaterialAttributes
   Base Color=LinearInterpolate(A=NamedRerouteUsage(), B=@n44.Base Color, Alpha=@n47)
   Normal=LinearInterpolate(A=NamedRerouteUsage(), B=@n44.Normal, Alpha=@n47)
  ) {attributeSetTypes [#2, #5]}
  n49 = GetMaterialAttributes(None=MaterialFunctionCall(NoRVT=@n43, RVT=StaticSwitchParameter(True=@n48, False=@n43) {group=RVT Blend, parameterName=UseSmoothRVTBlend}) {materialFunction=&FN_RVT_Switch_Material}.Result) {attributeGetTypes [#5]}
  n50 = Transform(None=@n49.Normal) {transformSourceType=TRANSFORMSOURCE_World, transformType=TRANSFORM_Tangent}
  n51 = MaterialFunctionCall!(
   In=SetMaterialAttributes!(MaterialAttributes=@n49.MaterialAttributes, Normal=MaterialProxyReplace(Realtime=@n49.Normal, MaterialProxy=@n50)) {attributeSetTypes [#5]}
   Enabled
   Disabled
   useDirectionalShading
   DirectionalShadingMask
  ) {materialFunction=/WildEstate/Environment/BaseMaterials/MaterialFunctions/MF_Test5050Unlit}
  MaterialFunctionCall!(
   In=MaterialFunctionCall(
    Material_In=MaterialFunctionCall!(In=@n51.Result) {materialFunction=/WildEstate/Environment/BaseMaterials/MaterialFunctions/MF_NucularGlowFields}.Result
    UseCrackedDamageTextures=StaticBool()
    UseAdditiveWPO=StaticBool() {value=true}
   ) {materialFunction=&MF_FN_DestructibleDamage}.Material_Out
   In_Translucent
   DestructibleDamageAmount
  ) {materialFunction=/WildEstate/Environment/BaseMaterials/MaterialFunctions/MF_WildEstate_Cracks}
  NamedRerouteDeclaration() {name=Test2, nodeColor=<rgba 0.125 1 0 1>}
  n52 = RuntimeVirtualTextureSampleParameter(World Position=WorldPosition().XYZ) {
   bEnableFeedback=false
   materialType=BaseColor_Normal_Specular_YCoCg
   parameterName=RVT_Base
   virtualTexture=/Game/Athena/Apollo/Maps/Landscape/VT/RVT_Apollo_Landscape
  }
  NamedRerouteDeclaration(None=@n52.Roughness) {name=RVT_Roughness, nodeColor=<rgba 0 0.7656 1 1>}
  NamedRerouteDeclaration(None=@n20) {name=GrassMaskUnclampt, nodeColor=<rgba 0.5234 1 0 1>}
  NamedRerouteDeclaration(None=@n22.Dirt_Masks) {name=GrassMask, nodeColor=<rgba 0.5234 1 0 1>}
  NamedRerouteDeclaration(None=@n52.Specular) {name=RVT_Specular, nodeColor=<rgba 0 0.7656 1 1>}
  NamedRerouteDeclaration(None=@n4) {name=RVT_HeightBlend, nodeColor=<rgba 0 0.7656 1 1>}
  NamedRerouteDeclaration(None=Multiply(A=@n52.BaseColor, B=Constant() {desc=Fix RVT MissMatch, r=1.02})) {name=RVT_BaseColor, nodeColor=<rgba 0 0.7656 1 1>}
  NamedRerouteDeclaration(None=Transform(None=NamedRerouteDeclaration(None=@n52.Normal) {name=RVT_Normal_WS, nodeColor=<rgba 0 0.7656 1 1>}) {transformSourceType=TRANSFORMSOURCE_World, transformType=TRANSFORM_Tangent}) {name=RVT_Normal_TS, nodeColor=<rgba 0 0.7656 1 1>}
  NamedRerouteUsage()
  NamedRerouteUsage()
  NamedRerouteUsage()
  NamedRerouteUsage()
  NamedRerouteUsage()
  SetMaterialAttributes!(MaterialAttributes, Base Color=@n21) {attributeSetTypes [#2]}
  StaticSwitchParameter(True=VertexColor().R, False=@n42) {group=GrassBlend, parameterName=UseTopLayerSnowBlendUseVTXColor}
  VertexColor()
 ]
}
