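# Unlit Wild Estate material for LED window glass on instanced meshes, with a dot-matrix look and a gameplay "Confidential" state.
# A collection parameter for that state blends in pulsing time-of-day colors and switches glass opacity.
# A UV-tiled rounded-box distance field forms the LED mask, while view angle modulates emissive intensity.
# Named reroutes link these to sunlight-shaded glass, world-aligned grunge and per-instance offsets, with an HLOD proxy color multiplier.
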Material /WildEstate/Environment/BaseMaterials/M_WildEstate_LED
graph {
 properties {
  bAutomaticallySetUsageInEditor=false
  bUseMaterialAttributes=true
  bUsedWithInstancedStaticMeshes=true
  physMaterial=/Game/FortressPhysicalMaterials/Glass
  shadingModel=MSM_Unlit
  usedShadingModels=Unlit
 }
 nodes [
  n1 = Desaturation(None=NamedRerouteUsage())
  n2 = MaterialFunctionCall(
   Color_Morning=Constant3Vector() {constant=<rgba 1.0435 10 0 1>}
   Color_Day=Constant3Vector() {constant=<rgba 0.3475 10 0 1>}
   Color_Evening=Constant3Vector() {constant=<rgba 0.5428 10 0 1>}
   Color_Night=Constant3Vector() {constant=<rgba 0.3973 10 0 1>}
  ) {materialFunction=&Color_Time_of_Day}
  n3 = ScalarParameter() {defaultValue=1.48, group=Window Gameplay, parameterName=Window_PulseLowVal}
  n4 = ScalarParameter() {defaultValue=1.5, group=Window Gameplay, parameterName=Window_PulseHighVal}
  n5 = Time()
  n6 = Sine(None=Divide(A=@n5) {constB=0.5})
  n7 = Sine(None=Divide(A=@n5) {constB=3})
  n8 = Min(A=Add(A=Floor(None=@n6)), B=Add(A=Floor(None=@n7)))
  n9 = LinearInterpolate(
   A=NamedRerouteUsage()
   B=LinearInterpolate(A=Multiply(A=@n1, B=@n2.Result), B=@n2.Result, Alpha=LinearInterpolate(A=@n3, B=@n4, Alpha=@n8) {constA=0.5, constB=5})
   Alpha=CollectionParameter() {collection=/WildEstate/TechArt/MPCs/MPC_WildEstate_TA, parameterName=Confidential}
  )
  LinearInterpolate(A=NamedRerouteUsage(), B=@n9, Alpha=ScalarParameter() {defaultValue=1, group=Window Gameplay, parameterName=Window_IsConfidential})
  n10 = VectorParameter() {parameterName=Glass_Color_Day, defaultValue=<rgba 0.5 0.4764 0.4 0.1451>}
  MaterialFunctionCall(
   Color_Morning=VectorParameter() {parameterName=Glass_Color_Morning, defaultValue=<rgba 0.45 0.408 0.36 0.1451>}.RGB
   Color_Day=@n10.RGB
   Color_Evening=VectorParameter() {parameterName=Glass_Color_Evening, defaultValue=<rgba 0.45 0.3835 0.36 0.1451>}.RGB
   Color_Night=VectorParameter() {parameterName=Glass_Color_Night, defaultValue=<rgba 0.4 0.3627 0.32 0.1451>}.RGB
  ) {materialFunction=&Color_Time_of_Day}
  n11 = VectorParameter() {parameterName=Reflection_Color_Day, defaultValue=<rgba 0.25 0.2323 0.175 0.1451>}
  MaterialFunctionCall(
   Color_Morning=VectorParameter() {parameterName=Reflection_Color_Morning, defaultValue=<rgba 0.25 0.215 0.175 0.1451>}.RGB
   Color_Day=@n11.RGB
   Color_Evening=VectorParameter() {parameterName=Reflection_Color_Evening, defaultValue=<rgba 0.25 0.2323 0.175 0.1451>}.RGB
   Color_Night=VectorParameter() {parameterName=Reflection_Color_Night, defaultValue=<rgba 0.25 0.215 0.175 0.1451>}.RGB
  ) {materialFunction=&Color_Time_of_Day}
  n12 = ScalarParameter() {defaultValue=0.3, group=Window Gameplay, parameterName=Window_SineTimeSpeed}
  n13 = ConstantBiasScale(None=Sine(None=Multiply(A=@n5, B=@n12)))
  n14 = Sine(None=Multiply(A=@n5) {constB=3})
  Min(A=Round(None=Saturate(None=@n13)), B=Saturate(None=ConstantBiasScale(None=@n14)))
  n15 = Constant() {r=0.8}
  n16 = Constant() {r=0.8}
  n17 = Constant() {r=0.8}
  n18 = Constant() {r=0.1}
  n19 = LinearInterpolate(
   A=Multiply(A=@n9) {constB=0.5}
   B=@n9
   Alpha=LinearInterpolate(
    A=ScalarParameter() {defaultValue=0.8, group=DotMatrixPixels, parameterName=Dot_Blend}
    B=MaterialFunctionCall(Value_Morning=@n15, Value_Day=@n16, Value_Evening=@n17, Value_Night=@n18) {materialFunction=&Scalar_Time_of_Day}.Result
    Alpha=CollectionParameter() {collection=&MPC_WildEstate_TA, parameterName=Confidential}
   ) {constA=0.985}
  ) {constAlpha=0.925}
  n20 = NamedRerouteUsage()
  n21 = ScalarParameter() {defaultValue=1, group=DotMatrixPixels, parameterName=UseInteriorDots, sortPriority=0}
  n22 = ScalarParameter() {defaultValue=1, group=DotMatrixPixels, parameterName=UseInteriorDots_Confidential, sortPriority=0}
  n23 = CollectionParameter() {collection=&MPC_WildEstate_TA, parameterName=Confidential}
  Multiply(
   A=LinearInterpolate(A=@n9, B=LinearInterpolate(A=@n19, B=@n9, Alpha=@n20), Alpha=LinearInterpolate(A=@n21, B=@n22, Alpha=@n23))
   B=ScalarParameter() {defaultValue=0.5, group=DotMatrixPixels, parameterName=InteriorDotsBoost, sortPriority=0}
  ) {constB=0.5}
  n24 = TextureObject() {texture=/WildEstate/Landscape/Textures/T_WildEstate_Detail_Asphalt}
  n25 = ScalarParameter() {defaultValue=256, parameterName=GrungeNoiseTiling}
  n26 = PreSkinnedPosition()
  n27 = VertexInterpolator(VS=@n26)
  n28 = StaticBoolParameter() {group=z_GameplayParameters_DestructibleDamage, parameterName=UseLocalSpace, sortPriority=100}
  n29 = StaticSwitch(
   True=Multiply(
    A=MaterialFunctionCall() {materialFunction=&ObjectScale}.Scale XYZ
    B=ShaderStageSwitch(PixelShader=@n27.PS, VertexShader=@n26)
   )
   False=WorldPosition().XYZ
   Value=@n28
  )
  n30 = PreSkinnedNormal()
  n31 = StaticSwitch(True=ShaderStageSwitch(PixelShader=VertexInterpolator(VS=@n30).PS, VertexShader=@n30), False=VertexNormalWS(), Value=@n28)
  NamedRerouteDeclaration(None=ComponentMask(None=MaterialFunctionCall(TextureObject=@n24, TextureSize=@n25, WorldPosition=@n29, World Space Normal=@n31) {materialFunction=&WorldAlignedTexture}.XYZ Texture) {r=true}) {name=WindowLines, nodeColor=<rgba 1 0 0.7734 1>}
  n32 = Constant3Vector() {constant=<rgba 100 10 1 0>}
  n33 = PerInstanceRandom()
  n34 = ObjectPositionWS()
  n35 = Constant() {r=0.01}
  n36 = Frac(None=Add(A=Multiply(A=@n32, B=@n33), B=Multiply(A=@n34, B=@n35)))
  n37 = DotProduct(A=ComponentMask(None=@n36) {g=true, r=true}, B=ComponentMask(None=@n36) {b=true, g=true})
  NamedRerouteDeclaration(None=Multiply(A=ConstantBiasScale(None=@n37) {bias=-0.5, scale=2}, B=Normalize(VectorInput=@n36))) {name=Offsets, nodeColor=<rgba 0.6172 1 0 1>}
  n38 = Add(A=OneMinus(None=NamedRerouteUsage())) {constB=0.15}
  n39 = Add(A=@n10.RGB, B=Multiply(A=@n11.RGB, B=Saturate(None=@n38)))
  n40 = CollectionParameter() {
   collection=&FortniteMaterialParameters
   parameterName=SunlightDirectionalVector
  }
  n41 = Multiply(A=VertexNormalWS(), B=TwoSidedSign())
  n42 = Saturate(None=Max(A=DotProduct(A=@n40, B=@n41), B=ScalarParameter() {defaultValue=0.5639, parameterName=MinDarkness}))
  NamedRerouteDeclaration(None=LinearInterpolate(
    A=@n10.RGB
    B=Multiply(A=@n39, B=@n42)
    Alpha=ScalarParameter() {defaultValue=1, group=Advanced, parameterName=UseDirectionalLightMultiplier, sortPriority=36}
   )) {name=BaseEmissive, nodeColor=<rgba 0.0781 1 0 1>}
  n43 = DotProduct(A=CameraVectorWS(), B=PixelNormalWS())
  n44 = LinearInterpolate(A=Power(Base=@n43) {constExponent=1}, B=Power(Base=@n43), Alpha=@n23)
  NamedRerouteDeclaration(None=MaterialFunctionCall(Emissive=LinearInterpolate(A=@n19, B=@n9, Alpha=@n44)) {materialFunction=&MF_FX_EmissiveAdjustments}.Result) {name=Emissive, nodeColor=<rgba 0 1 0.9063 1>}
  n45 = ScalarParameter() {defaultValue=0.5, group=DotMatrixPixels, parameterName=Dot_Smoothness}
  n46 = Multiply(A=@n45) {constB=-1}
  n47 = TextureCoordinate()
  n48 = AppendVector(
   A=ScalarParameter() {defaultValue=128, group=DotMatrixPixels, parameterName=Dot_XAmount}
   B=ScalarParameter() {defaultValue=1, group=DotMatrixPixels, parameterName=Dot_YAmount}
  )
  n49 = Subtract(A=Frac(None=Multiply(A=@n47, B=@n48))) {constB=0.5}
  n50 = ScalarParameter() {group=DotMatrixPixels, parameterName=Dot_Radius}
  n51 = Add(A=Subtract(A=Abs(None=@n49), B=ScalarParameter() {defaultValue=0.01, group=DotMatrixPixels, parameterName=Dot_Size}), B=@n50)
  n52 = Min(A=Max(A=ComponentMask(None=@n51) {r=true}, B=ComponentMask(None=@n51) {g=true}) {constB=0}) {constB=0}
  n53 = Length(None=Max(A=@n51) {constB=0}) {is=class MaterialExpressionLength}
  n54 = Multiply(A=Subtract(A=Add(A=@n52, B=@n53), B=@n50)) {constB=-1}
  NamedRerouteDeclaration(None=Saturate(None=SmoothStep(Min=@n46, Max=@n45, Value=@n54))) {name=LED, nodeColor=<rgba 0 1 0.8828 1>}
  n55 = Add(A=NamedRerouteUsage(), B=ScalarParameter() {defaultValue=0.025, parameterName=Base Color Add})
  n56 = Multiply(A=@n55, B=ScalarParameter() {defaultValue=0.8, parameterName=Base Color Mult}) {constB=0.2}
  n57 = Constant() {r=1}
  n58 = Saturate(None=Subtract(A=Desaturation(None=@n55, Fraction=@n57), B=ScalarParameter() {defaultValue=0.2, parameterName=Base Color Threshold}) {constB=0.45})
  n59 = ComponentMask(None=VertexNormalWS()) {r=true}
  n60 = Multiply(A=Add(A=Abs(None=@n59)) {constB=-0.5}, B=ScalarParameter() {defaultValue=0.2, parameterName=Corner Break}) {constB=0.12}
  NamedRerouteDeclaration(None=Multiply(A=LinearInterpolate(A=@n55, B=@n56, Alpha=@n58), B=Add(A=@n60))) {name=BaseGlass, nodeColor=<rgba 0 0.0391 1 1>}
  n61 = NamedRerouteUsage()
  n62 = MaterialProxyReplace(Realtime=@n61, MaterialProxy=Multiply(A=@n61, B=VectorParameter() {parameterName=HLOD_ColorMult, defaultValue=<rgba 1.5 1.5 1.5 1>}.RGB))
  n63 = CollectionParameter() {
   collection=&FortniteMaterialParameters
   parameterName=SunlightDirectionalVector
  }
  n64 = ComponentMask(None=@n63) {b=true, g=true, r=true}
  n65 = Transform(None=Constant3Vector() {constant=<rgba 0 0 -1 0>}) {transformSourceType=TRANSFORMSOURCE_Camera}
  SetMaterialAttributes!(
   MaterialAttributes
   Base Color=@n62
   Emissive Color=@n62
   Opacity=LinearInterpolate(
    A=ScalarParameter() {defaultValue=0.01, parameterName=Opacity_Glass}
    B=ScalarParameter() {defaultValue=1, parameterName=Opacity_GlassConfidential}
    Alpha=CollectionParameter() {collection=&MPC_WildEstate_TA, parameterName=Confidential}
   )
   Normal=Normalize(VectorInput=Add(A=@n64, B=@n65) {desc=Compromise for Direction light and all othe rlights})
   Roughness=VertexColor().R
  ) {attributeSetTypes [
    #1
    #2
    #3
    #4
    #5
   ]}
  n66 = Multiply(A=NamedRerouteUsage(), B=Constant() {r=0.01})
  SingleLayerWaterMaterialOutput!(ScatteringCoefficients=@n66, AbsorptionCoefficients=@n66, PhaseG, ColorScaleBehindWater)
 ]
}
