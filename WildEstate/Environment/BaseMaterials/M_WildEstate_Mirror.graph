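# Default Lit mirror material for WildEstate environments that fakes glass reflections instead of rendering real ones.
# A window texture is projected on a camera-aligned virtual plane shifted along the reflection vector, with per-instance UV variation.
# Its tinted lines plus time-of-day glass color drive exposure-compensated emissive, then pass through destructible-damage and crack functions.
# Sunlight, absorption, scatter and opacity-blend nodes are left unconnected.
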Material /WildEstate/Environment/BaseMaterials/M_WildEstate_Mirror
graph {
 properties {bAutomaticallySetUsageInEditor=false, bTangentSpaceNormal=false, bUseMaterialAttributes=true, usedShadingModels=Default Lit}
 nodes [
  CameraVectorWS()
  n1 = CollectionParameter() {
   collection=&FortniteMaterialParameters
   parameterName=SunlightDirectionalVector
  }
  n2 = DotProduct(A=@n1, B=Multiply(A=VertexNormalWS(), B=TwoSidedSign()))
  Divide(A=Add(A=@n2))
  n3 = Subtract(A=SceneDepthWithoutWater!(UVs).R, B=PixelDepth().R)
  Divide(A=ScalarParameter() {parameterName=Absorption}, B=@n3)
  n4 = VectorParameter() {group=SingleLayerShading, parameterName=Absorption, defaultValue=<rgba 0.0084 0.026 0.9584 149.7252>}
  Divide(A=Divide(A=Constant() {r=1}, B=Multiply(A=@n4.RGB, B=@n4.RGB)), B=@n3)
  n5 = TransformPosition(None=Constant3Vector()) {transformSourceType=TRANSFORMPOSSOURCE_Instance, transformType=TRANSFORMPOSSOURCE_World}
  n6 = MaterialFunctionCall(Float3=Divide(A=@n5) {constB=100}) {materialFunction=&BreakOutFloat3Components}
  n7 = Frac(None=Add(A=Add(A=@n6.R, B=@n6.R), B=@n6.B))
  n8 = Transform(None=Constant3Vector() {constant=<rgba 0 0 1 0>}) {transformSourceType=TRANSFORMSOURCE_Camera}
  n9 = Transform(None=Constant3Vector() {constant=<rgba 1 0 0 0>}) {transformSourceType=TRANSFORMSOURCE_Camera}
  n10 = ObjectPositionWS()
  n11 = ReflectionVectorWS()
  n12 = Add(A=@n10, B=Multiply(A=Multiply(A=@n11) {constB=-1}, B=ScalarParameter() {defaultValue=15, parameterName=GlassRelectionOffset}))
  n13 = ScalarParameter() {defaultValue=222.7369, parameterName=GlassRelectionTiling}
  n14 = TextureSample(UVs=Add(
    A=Multiply(A=@n7) {constB=33.3}
    B=MaterialFunctionCall(Plane Normal Axis=@n8, Plane X Axis=@n9, Plane Center=@n12, UV Scale=@n13) {materialFunction=&VirtualPlaneCoordinates}.UVs
   )) {texture=/WildEstate/Environment/Textures/T_WildEstate_Windows}
  LinearInterpolate(
   A=ScalarParameter() {defaultValue=0.28, parameterName=Opacity_Lines}
   B=ScalarParameter() {defaultValue=0.664, parameterName=Opacity_Glass}
   Alpha=@n14.R
  )
  n15 = VectorParameter() {parameterName=BaseColorReflections, defaultValue=<rgba 0.4052 0.6545 0.8229 1>}
  LinearInterpolate(A=@n15.RGB, B=VectorParameter() {parameterName=BaseColorOpaque, defaultValue=<rgba 1 1 1 1>}.RGB, Alpha=@n14.R)
  n16 = Multiply(A=Multiply(A=@n15.RGB, B=OneMinus(None=@n14.R)), B=ScalarParameter() {defaultValue=0.5208, parameterName=EmissiveReflectionLines})
  n17 = MaterialFunctionCall(
   Color_Morning=VectorParameter() {parameterName=GlassColor_Morning, defaultValue=<rgba 0 0.1996 0.4792 1>}.RGB
   Color_Day=VectorParameter() {parameterName=GlassColor_Day, defaultValue=<rgba 0 0.1931 0.4635 1>}.RGB
   Color_Evening=VectorParameter() {parameterName=GlassColor_Evening, defaultValue=<rgba 0 0.1996 0.4792 1>}.RGB
   Color_Night=VectorParameter() {parameterName=GlassColor_night, defaultValue=<rgba 0 0.1671 0.401 1>}.RGB
  ) {materialFunction=&Color_Time_of_Day}
  n18 = SetMaterialAttributes!(
   MaterialAttributes
   Base Color=ScalarParameter() {defaultValue=0.28, parameterName=ReflectionBrightness}
   Specular=Constant()
   Roughness=ScalarParameter() {parameterName=Roughness}
   Emissive Color=EyeAdaptationInverse(LightValueInput=Add(A=@n16, B=@n17.Result)).EyeAdaptationInverse
   Opacity
   Normal
   Metallic=Constant()
  ) {attributeSetTypes [
    #1
    #2
    #3
    #4
    #5
    #6
    #7
   ]}
  MaterialFunctionCall!(
   In
   In_Translucent=MaterialFunctionCall(Material_In=@n18, UseCrackedDamageTextures=StaticBool(), UseAdditiveWPO=StaticBool() {value=true}) {materialFunction=&MF_FN_DestructibleDamage}.Material_Out
   DestructibleDamageAmount
  ) {materialFunction=/WildEstate/Environment/BaseMaterials/MaterialFunctions/MF_WildEstate_Cracks}
  n19 = VectorParameter() {group=SingleLayerShading, parameterName=Scatter, defaultValue=<rgba 0.257 0.2539 0.6632 -0.2284>}
  Multiply(A=@n19.RGB, B=Divide(A=@n19.A) {constB=100})
  Normalize(VectorInput=Subtract(A=@n5, B=CameraPositionWS()))
  Normalize(VectorInput=Subtract(A=@n10, B=CameraPositionWS()))
  Saturate(None=Max(A=@n2, B=ScalarParameter() {defaultValue=0.4109, parameterName=MinDarkness}))
  ScalarParameter() {parameterName=Spec}
  ScalarParameter() {parameterName=Metallic}
  ScalarParameter() {parameterName=Scatter}
  TextureCoordinate()
 ]
}
