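# Base road material for WildEstate spline and Nanite meshes, layering asphalt, painted markings and optional gravel.
# Asphalt color takes a close-up landscape multiply modulated by vertex blue; mask green blends in markings, optional mask red blends gravel.
# The result passes through MF_Test5050Unlit, optional roughness scaling and MF_Curie_Landscape.
# Normals are world-space, converted to tangent for material proxies, and an optional power-curve WPO bends the mesh along local position.
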Material /WildEstate/Environment/BaseMaterials/M_WildEstate_Roads_Base
graph {
 properties {
  bAutomaticallySetUsageInEditor=false
  bTangentSpaceNormal=false
  bUseMaterialAttributes=true
  bUsedWithNanite=true
  bUsedWithSplineMeshes=true
  physMaterial=/Game/FortressPhysicalMaterials/Asphalt
  usedShadingModels=Default Lit
 }
 nodes [
  n1 = GetMaterialAttributes(None=MaterialFunctionCall!(In Material, LayerWeight, ArrayTextureUV) {materialFunction=/WildEstate/Environment/BaseMaterials/MaterialFunctions/MF_WildEstate_Asphalt}.Out StaticMesh) {attributeGetTypes [#1]}
  n2 = Multiply(
   A=@n1.Base Color
   B=MaterialFunctionCall() {materialFunction=/WildEstate/Environment/BaseMaterials/MaterialFunctions/MF_WildEstate_LandscapeCloseUpTexture}.Out Material
  )
  n3 = Saturate(None=LinearInterpolate(
    A=ScalarParameter() {defaultValue=-4.9961, parameterName=TransitionEdges_Low}
    B=ScalarParameter() {defaultValue=2.8462, parameterName=TransitionEdges_High}
    Alpha=VertexColor().B
   ))
  n4 = ScalarParameter() {defaultValue=0.9, parameterName=RoadMarkingBrightness}
  n5 = VectorParameter() {group=RoadMarkings, parameterName=Road_Color, defaultValue=<rgba 1 1 1 1>}
  n6 = TextureSampleParameter2D() {group=RoadMarkings, parameterName=Road_Diffuse, texture=/WildEstate/Environment/Terrain/Roads/Textures/T_WildEstate_AsphaltRoad1_D}
  n7 = StaticSwitchParameter(True=Multiply(A=@n5.RGB, B=@n6.RGB), False=@n5.RGB) {defaultValue=true, group=RoadMarkings, parameterName=UseDiffuseTexture}
  n8 = Clamp(None=@n7, Max=ScalarParameter() {defaultValue=0.6, parameterName=RoadMarkingMaxBrightness})
  n9 = TextureSampleParameter2D() {
   group=RoadMarkings
   parameterName=Road_Mask
   samplerType=SAMPLERTYPE_Masks
   texture=/Game/Athena/Apollo/Environments/Landscape/Textures/Roads/AsphaltRoad1_rgb
  }
  n10 = BlendMaterialAttributes(
   A=SetMaterialAttributes!(MaterialAttributes=@n1.MaterialAttributes, Base Color=LinearInterpolate(A=@n2, B=@n1.Base Color, Alpha=@n3)) {attributeSetTypes [#1]}
   B=SetMaterialAttributes!(MaterialAttributes, Base Color=Multiply(A=@n4, B=@n8), Specular=Constant(), Roughness=Constant() {r=0.95}) {attributeSetTypes [#1, #2, #3]}
   Alpha=@n9.G
  )
  n11 = BlendMaterialAttributes(
   A=@n10
   B=MaterialFunctionCall!(In Material, LayerWeight, Additional LayerWeight for Transitions, ArrayTextureUV) {materialFunction=/WildEstate/Environment/BaseMaterials/MaterialFunctions/MF_WildEstate_Gravel2}.StaticMesh
   Alpha=NamedRerouteUsage()
  )
  n12 = GetMaterialAttributes(None=MaterialFunctionCall!(
    In=StaticSwitchParameter(True=@n11, False=@n10) {group=RoadMarkings, parameterName=Use Mask Red Channel}
    Enabled
    Disabled
    useDirectionalShading
    DirectionalShadingMask
   ) {materialFunction=/WildEstate/Environment/BaseMaterials/MaterialFunctions/MF_Test5050Unlit}.Result) {attributeGetTypes [#3]}
  n13 = Multiply(A=@n12.Roughness, B=ScalarParameter() {defaultValue=0.5, group=RoadMarkings, parameterName=Set Custom Roughness})
  n14 = GetMaterialAttributes(None=SetMaterialAttributes!(
    MaterialAttributes=@n12.MaterialAttributes
    Roughness=StaticSwitchParameter(True=@n13, False=@n12.Roughness) {group=RoadMarkings, parameterName=Use Custom Roughness}
   ) {attributeSetTypes [#3]}) {attributeGetTypes [#4]}
  n15 = MaterialProxyReplace(Realtime=@n14.Normal, MaterialProxy=Transform(None=@n14.Normal) {transformSourceType=TRANSFORMSOURCE_World, transformType=TRANSFORM_Tangent})
  n16 = LocalPosition()
  n17 = Divide(
   A=Abs(None=ChannelMaskParameter(None=@n16.XYZ) {group=WPO Curve, parameterName=WPO Channel}.R)
   B=ScalarParameter() {defaultValue=350, group=WPO Curve, parameterName=WPO Curve Length}
  ) {constB=350}
  n18 = ScalarParameter() {defaultValue=2, group=WPO Curve, parameterName=WPO Curve Exp}
  n19 = Constant()
  n20 = Multiply(
   A=Saturate(None=Power(Base=@n17, Exp=@n18) {constExponent=8})
   B=MaterialFunctionCall(R=@n19, G=@n19, B=ScalarParameter() {defaultValue=-6, group=WPO Curve, parameterName=WPO Curve Offset}) {materialFunction=&AppendMany}.RGB
  )
  n21 = Constant()
  MaterialFunctionCall!(
   In=SetMaterialAttributes!(
    MaterialAttributes=SetMaterialAttributes!(MaterialAttributes=@n14.MaterialAttributes, Normal=@n15) {attributeSetTypes [#4]}
    World Position Offset=StaticSwitchParameter(True=@n20, False=@n21) {group=WPO Curve, parameterName=WPO Curve, sortPriority=30}
   ) {attributeSetTypes [#5]}
   UseHeightFalloff=StaticBool()
   OnFireEmissiveOverride
   HeightFalloffRadius
   HeightFalloffHardness
   VariationMasksMobile
   In_Grass
   Is_Stylised_WE=StaticBool() {value=true}
  ) {materialFunction=/Game/Packages/Fortress_SharedMaterials/Base_Material_Functions/MF_Curie_Landscape}
  n22 = TextureSampleParameter2D() {
   group=RoadMarkings
   parameterName=Road_Outlines
   samplerType=SAMPLERTYPE_Masks
   texture=/WildEstate/Environment/TestContent/T_WildEstate_Detail_ConcreteNoise_A_A_D
  }
  Multiply(A=@n7, B=@n22.R)
  NamedRerouteDeclaration(None=@n9.R) {name=Mask R, nodeColor=<rgba 1 0 0.0245 1>}
 ]
}
