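# Unlit WildEstate sky material: a zenith-to-horizon gradient with stars and a storm that follows safe-zone progress.
# Panned radial noise breaks up gradient and storm edges; a virtual overhead plane adds the storm's top ring.
# Storm masks use safe-zone collection parameters, while time-of-day functions drive storm brightness and night stars.
# Reflection-capture switches substitute a boosted sky with ground color and quantized fake sun, and flatten below-zero geometry.
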Material /WildEstate/Environment/BaseMaterials/M_WildEstate_Sky
graph {
 properties {bAutomaticallySetUsageInEditor=false, bIsSky=true, shadingModel=MSM_Unlit, usedShadingModels=Unlit}
 nodes [
  n1 = CollectionParameter() {collection=/Game/Athena/SafeZone/MPC_SafeZoneClouds, parameterName=SafeZoneLoc}
  n2 = MaterialFunctionCall(
   Plane Normal Axis=Constant3Vector() {constant=<rgba 0 0 1 0>}
   Plane X Axis=Constant3Vector() {constant=<rgba 1 0 0 0>}
   Plane Center=AppendVector(A=ComponentMask(None=@n1) {g=true, r=true}, B=ScalarParameter() {defaultValue=158411.5, parameterName=VirtualPlaneHeight})
   UV Scale=ScalarParameter() {defaultValue=187429.7344, parameterName=2ndNoiseSize}
  ) {materialFunction=&VirtualPlaneCoordinates}
  n3 = Constant2Vector() {g=0.5, r=0.5}
  n4 = Multiply(
   A=MaterialFunctionCall(Vector or UVs=Subtract(A=@n2.UVs, B=@n3)) {materialFunction=&VectorToRadialValue}.Radial Coordinates
   B=AppendVector(A=ScalarParameter() {defaultValue=3, parameterName=2ndNoiseSize_Radial}, B=Constant() {r=1})
  )
  AppendVector(A=@n4)
  ComponentMask(None=TextureCoordinate() {coordinateIndex=1}) {g=true}
  n5 = VectorParameter() {parameterName=Color_Zenith, defaultValue=<rgba 0 0.1654 0.5 1>}
  n6 = VectorParameter() {parameterName=Color_Horizon, defaultValue=<rgba 0.3206 1.151 1.4 1>}
  n7 = NamedRerouteUsage()
  n8 = Add(
   A=NamedRerouteUsage()
   B=Multiply(A=LinearInterpolate(A=@n5.RGB, B=@n6.RGB, Alpha=@n7), B=ScalarParameter() {defaultValue=1.4, parameterName=Brightness})
  )
  n9 = CameraVectorWS()
  n10 = Saturate(None=Multiply(A=ComponentMask(None=@n9) {b=true}) {constB=400})
  n11 = Multiply(A=Desaturation(None=@n8, Fraction=ScalarParameter() {defaultValue=-0.1, parameterName=Saturation}), B=OneMinus(None=@n10))
  LinearInterpolate(A=@n11, B=NamedRerouteUsage(), Alpha=NamedRerouteUsage())
  n12 = Multiply(A=NamedRerouteUsage(), B=ScalarParameter() {defaultValue=-0.14, parameterName=StormNoiseSubtract})
  n13 = NamedRerouteUsage()
  n14 = Add(A=Multiply(A=NamedRerouteUsage(), B=ScalarParameter() {defaultValue=268.6616, parameterName=StormNoiseadd}), B=LocalPosition().Z)
  n15 = LinearInterpolate(
   A=ScalarParameter() {defaultValue=-2675.3301, parameterName=StormTop_Bias_Start}
   B=ScalarParameter() {defaultValue=2318.4714, parameterName=StormTop_Bias}
   Alpha=NamedRerouteUsage()
  )
  n16 = ScalarParameter() {defaultValue=8962.7988, parameterName=StormTop_Bias_End}
  n17 = Saturate(None=Multiply(A=NamedRerouteUsage()) {constB=3})
  n18 = ScalarParameter() {defaultValue=1810.8776, parameterName=StormTop_Distance_Start}
  n19 = ScalarParameter() {defaultValue=2579.6116, parameterName=StormTop_Distance}
  n20 = NamedRerouteUsage()
  n21 = Divide(
   A=Subtract(A=@n14, B=LinearInterpolate(A=@n15, B=@n16, Alpha=@n17))
   B=LinearInterpolate(
    A=LinearInterpolate(A=@n18, B=@n19, Alpha=@n20)
    B=ScalarParameter() {defaultValue=2420.488, parameterName=StormTop_Distance_End}
    Alpha=NamedRerouteUsage()
   )
  )
  n22 = Saturate(None=Power(Base=OneMinus(None=@n21), Exp=ScalarParameter() {defaultValue=1.3855, parameterName=StormTop_Falloff}))
  n23 = Subtract(
   A=@n14
   B=LinearInterpolate(
    A=ScalarParameter() {defaultValue=-1635.5826, parameterName=StormBottom_Bias_Start}
    B=ScalarParameter() {defaultValue=-374.7981, parameterName=StormBottom_Bias}
    Alpha=NamedRerouteUsage()
   )
  )
  n24 = ScalarParameter() {defaultValue=846.5688, parameterName=StormBottom_Distance}
  n25 = Saturate(None=Power(Base=Divide(A=@n23, B=@n24), Exp=ScalarParameter() {defaultValue=2.1537, parameterName=StormBottom_Falloff}))
  n26 = Saturate(None=Add(A=LinearInterpolate(A=@n13, Alpha=@n22) {constB=0}, B=Multiply(A=@n22, B=@n25)))
  NamedRerouteDeclaration(None=Saturate(None=Add(A=@n12, B=@n26))) {name=StormMask, nodeColor=<rgba 1 0 0.2344 1>}
  n27 = ComponentMask(None=CameraVectorWS()) {b=true}
  n28 = Power(Base=OneMinus(None=Abs(None=@n27)), Exp=ScalarParameter() {defaultValue=10, parameterName=GradientFalloff})
  n29 = NamedRerouteDeclaration(None=@n28) {name=Linear Gradient, nodeColor=<rgba 0 0.0625 1 1>}
  n30 = WorldPosition() {worldPositionShaderOffset=WPT_ExcludeAllShaderOffsets}
  n31 = ComponentMask(None=Subtract(A=@n30.XYZ, B=ObjectPositionWS())) {g=true, r=true}
  n32 = ScalarParameter() {defaultValue=1680410.125, parameterName=UVScaleHeight}
  n33 = AppendVector(
   A=Multiply(
    A=MaterialFunctionCall(Vector or UVs=@n31) {materialFunction=&VectorToRadialValue}.Vector Converted to Angle
    B=ScalarParameter() {defaultValue=4, parameterName=UVScaleHorizontal}
   )
   B=ComponentMask(None=Divide(A=@n30.XYZ, B=@n32)) {b=true}
  )
  n34 = AppendVector(A=ScalarParameter() {defaultValue=-0.005, parameterName=PanSpeed_Noise_x}, B=ScalarParameter() {parameterName=PanSpeed_Noise_y})
  n35 = TextureCoordinate()
  n36 = ComponentMask(None=TextureSample!(UVs=Panner(Coordinate=@n33, Speed=@n34), Tex, "DDX(UVs)"=DDX(Value=@n35), "DDY(UVs)"=DDY(Value=@n35), Apply View MipBias) {mipValueMode=TMVM_Derivative, texture=/WildEstate/Environment/Textures/WildEstate_Sky_M}.RGB) {r=true}
  n37 = Multiply(A=@n36, B=ScalarParameter() {defaultValue=0.312, parameterName=Noiseadd})
  NamedRerouteDeclaration(None=Saturate(None=Subtract(A=@n29, B=@n37))) {name=Gradient, nodeColor=<rgba 0.1484 0 1 1>}
  NamedRerouteDeclaration(None=@n36) {name=Noise, nodeColor=<rgba 0 1 0.625 1>}
  n38 = OneMinus(None=Saturate(None=CollectionParameter() {collection=&MPC_SafeZoneClouds, parameterName=SafeZoneScale}))
  NamedRerouteDeclaration(None=Saturate(None=Power(Base=@n38))) {name=ProgressNormalized, nodeColor=<rgba 1 0 0.375 1>}
  n39 = LinearInterpolate(
   A=CollectionParameter() {collection=&MPC_SafeZoneClouds, parameterName=CloudStartingBias}
   Alpha=CollectionParameter() {collection=&MPC_SafeZoneClouds, parameterName=PreventStormShrinking}
  ) {constA=1, constAlpha=1}
  NamedRerouteDeclaration(None=OneMinus(None=Saturate(None=@n39))) {name=ProgressStartupNormalized, nodeColor=<rgba 1 0 0.375 1>}
  n40 = Panner(
   Coordinate=@n4
   Speed=AppendVector(A=ScalarParameter() {defaultValue=0.02, parameterName=PanSpeed_TopPlane_x}, B=ScalarParameter() {parameterName=PanSpeed_TopPlane_y})
  )
  n41 = TextureCoordinate()
  n42 = DDX(Value=@n41)
  n43 = DDY(Value=@n41)
  n44 = Add(
   A=Distance(A=ComponentMask(None=@n2.World Position - Centered) {g=true, r=true}, B=Constant2Vector())
   B=Multiply(
    A=TextureSample(UVs=@n40, "DDX(UVs)"=@n42, "DDY(UVs)"=@n43) {mipValueMode=TMVM_Derivative, texture=&WildEstate_Sky_M}.R
    B=ScalarParameter() {defaultValue=17988.2461, parameterName=StormNoiseadd2}
   )
  )
  n45 = Saturate(None=CollectionParameter() {collection=&MPC_SafeZoneClouds, parameterName=SafeZoneScale})
  n46 = LinearInterpolate(
   A=ScalarParameter() {defaultValue=29031.1191, parameterName=SafeZoneRadius_Max}
   B=ScalarParameter() {defaultValue=215773.875, parameterName=SafeZoneRadius_Min}
   Alpha=@n45
  )
  n47 = ScalarParameter() {defaultValue=52028.6992, parameterName=SafeZoneRadius_BlendDist}
  n48 = ScalarParameter() {defaultValue=46531.0312, parameterName=SafeZoneRadius_BlendDist_Min}
  n49 = Power(
   Base=Divide(A=Subtract(A=@n44, B=@n46), B=LinearInterpolate(A=@n47, B=@n48, Alpha=@n45))
   Exp=ScalarParameter() {defaultValue=0.85, parameterName=SafeZoneRadius_Falloff}
  )
  n50 = Divide(
   A=Subtract(A=OneMinus(None=@n45), B=ScalarParameter() {defaultValue=0.5, parameterName=OffsetTopRingTiming})
   B=ScalarParameter() {defaultValue=0.196, parameterName=OffsetTopRingTimingMax}
  )
  NamedRerouteDeclaration(None=LinearInterpolate(B=Saturate(None=@n49), Alpha=Saturate(None=@n50)) {constB=0}) {name=TopStorm Virtual Plane, nodeColor=<rgba 1 0 0.8672 1>}
  n51 = Multiply(A=TextureCoordinate(), B=ScalarParameter() {defaultValue=3.8, parameterName=StarsTiling})
  n52 = Constant()
  n53 = MaterialFunctionCall(Value_Morning=@n52, Value_Day=@n52, Value_Evening=@n52, Value_Night=ScalarParameter() {defaultValue=3, parameterName=StarsBrightness}) {materialFunction=&Scalar_Time_of_Day}
  n54 = Saturate(None=Divide(A=LocalPosition().Z, B=ScalarParameter() {defaultValue=21729, parameterName=StarsBrightnessGradientFade}))
  NamedRerouteDeclaration(None=Multiply(
    A=TextureSample!(UVs=@n51, Tex, Apply View MipBias) {samplerType=SAMPLERTYPE_Masks, texture=/Game/Packages/Fortress_Sky/Textures/T_Sky_Stars_Sharp}.RGB
    B=Multiply(A=@n53.Result, B=@n54)
   )) {name=Stars, nodeColor=<rgba 0 1 0.8125 1>}
  NamedRerouteDeclaration(None=Min(A=NamedRerouteUsage(), B=ScalarParameter() {defaultValue=1, parameterName=MaxStorm})) {name=Storm Alpha, nodeColor=<rgba 0.8984 0 1 1>}
  n55 = ScalarParameter() {defaultValue=0.8, parameterName=Storm_Brightness_Morning}
  n56 = ScalarParameter() {defaultValue=1, parameterName=Storm_Brightness_Day}
  n57 = ScalarParameter() {defaultValue=0.8, parameterName=Storm_Brightness_Evening}
  n58 = ScalarParameter() {defaultValue=0.5, parameterName=Storm_Brightness_Night}
  n59 = LinearInterpolate(
   A=LinearInterpolate(
    A=VectorParameter() {parameterName=StromColor_Start, defaultValue=<rgba 0.272 0.2393 0.5469 1>}.RGB
    B=VectorParameter() {parameterName=StromColor, defaultValue=<rgba 0.3329 0.1289 0.9167 1>}.RGB
    Alpha=NamedRerouteUsage()
   )
   B=VectorParameter() {parameterName=StromColorB, defaultValue=<rgba 0.0635 0.0084 0.1406 1>}.RGB
   Alpha=NamedRerouteUsage()
  )
  n60 = VectorParameter() {parameterName=StromColorTop, defaultValue=<rgba 0.4748 0.1484 1.5 1>}
  n61 = NamedRerouteUsage()
  NamedRerouteDeclaration(None=Multiply(
    A=MaterialFunctionCall(Value_Morning=@n55, Value_Day=@n56, Value_Evening=@n57, Value_Night=@n58) {materialFunction=&Scalar_Time_of_Day}.Result
    B=LinearInterpolate(A=@n59, B=@n60.RGB, Alpha=@n61)
   )) {name=Storm Color, nodeColor=<rgba 0 1 0.5781 1>}
  n62 = ScalarParameter() {group=Reflection Pass, parameterName=Reflection Desaturation}
  n63 = Multiply(
   A=Multiply(
    A=Desaturation(None=@n8, Fraction=@n62) {luminanceFactors=<rgba 0.2126 0.7152 0.0722 1>}
    B=ScalarParameter() {defaultValue=3, group=Reflection Pass, parameterName=Reflection Boost}
   )
   B=ScalarParameter() {defaultValue=1, group=Reflection Pass, parameterName=Reflection Secondary Boost}
  )
  n64 = MaterialFunctionCall(
   Color_Morning=VectorParameter() {parameterName=Ground Morning, defaultValue=<rgba 0.2296 0.256 0.2812 1>}.RGB
   Color_Day=VectorParameter() {parameterName=Ground Day, defaultValue=<rgba 0.4219 0.2909 0.1868 1>}.RGB
   Color_Evening=VectorParameter() {parameterName=Ground Evening, defaultValue=<rgba 0.2865 0.1997 0.155 1>}.RGB
   Color_Night=VectorParameter() {parameterName=Ground Night, defaultValue=<rgba 0.2572 0.2115 0.2865 1>}.RGB
  ) {materialFunction=&Color_Time_of_Day}
  n65 = Multiply(A=DotProduct(A=CameraVectorWS(), B=SkyAtmosphereLightDirection())) {constB=-1}
  n66 = VectorParameter() {parameterName=Fake Sun Morning, defaultValue=<rgba 1 0.7133 0.5938 1>}
  n67 = VectorParameter() {parameterName=Fake Sun Day, defaultValue=<rgba 1.3 1.1267 1.0562 1>}
  n68 = VectorParameter() {parameterName=Fake Sun Evening, defaultValue=<rgba 1 0.8584 0.8177 1>}
  n69 = VectorParameter() {parameterName=Fake Sun Night, defaultValue=<rgba 1.0125 1.1135 1.8 1>}
  n70 = Floor(None=Multiply(
    A=Saturate(None=@n65)
    B=MaterialFunctionCall(Color_Morning=@n66.RGB, Color_Day=@n67.RGB, Color_Evening=@n68.RGB, Color_Night=@n69.RGB) {materialFunction=&Color_Time_of_Day}.Result
   ) {constB=10})
  n71 = ScalarParameter() {defaultValue=1, parameterName=Fake Light Intensity}
  ReflectionCapturePassSwitch!(Default=@n11, Reflection=Add(A=LinearInterpolate(A=@n63, B=@n64.Result, Alpha=@n10), B=Multiply(A=@n70, B=@n71)))
  n72 = WorldPosition()
  n73 = Constant3Vector() {constant=<rgba 0 0 -0.999 0>}
  n74 = Multiply(A=@n72.Z) {constB=-1}
  ReflectionCapturePassSwitch(Default=Constant(), Reflection=Multiply(A=Multiply(A=@n72.XYZ, B=@n73), B=Saturate(None=@n74)))
  Reroute(None=@n28)
  Saturate(None=ScalarParameter() {defaultValue=0.3249, desc=Storm Progress, parameterName=ProgressNormalized})
 ]
}
