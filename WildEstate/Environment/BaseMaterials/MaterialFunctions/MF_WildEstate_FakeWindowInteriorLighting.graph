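# Material function that fakes lit interiors behind window glass, outputting a time-of-day tinted color, glass alpha multiplier and depth mask.
# The mask combines InteriorCubemap depth, sphere masking and view-angle falloff, with a cheaper bump-offset sphere at Low/Medium quality.
# Custom primitive data enables lighting and swaps the lit face of two-sided glass; flicker and camera-distance fade are also computed.
# Spherical UV, gradient/fresnel switch and four-wave flicker branches are left unconnected.
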MaterialFunction /WildEstate/Environment/BaseMaterials/MaterialFunctions/MF_WildEstate_FakeWindowInteriorLighting
graph {nodes [
  n1 = Divide(A=ConstantBiasScale(None=TextureCoordinate()) {bias=-0.5, scale=2}, B=Constant2Vector() {desc=Size, g=1, r=1})
  n2 = Multiply(A=@n1, B=Constant2Vector() {desc=Curvature, g=1, r=1})
  n3 = DotProduct(A=@n2, B=@n2)
  n4 = SquareRoot(None=OneMinus(None=Saturate(None=@n3)))
  AppendVector(A=@n4, B=@n2)
  Ceil(None=@n4)
  Constant2Vector() {g=0.5, r=0.5}
  n5 = NamedRerouteUsage()
  n6 = TextureCoordinate()
  n7 = StaticSwitchParameter(
   True=StaticSwitchParameter(True=TextureCoordinate() {coordinateIndex=2}, False=TextureCoordinate() {coordinateIndex=1}) {group=Interior Lighting, parameterName=Emissive Depth UV2, sortPriority=-30}
   False=@n6
  ) {desc="Default is UV0 ", group=Interior Lighting, parameterName=Emissive Depth UV1, sortPriority=-30}
  n8 = MaterialFunctionCall!(
   UVs=@n7
   Tiling=AppendVector(
    A=ScalarParameter() {defaultValue=1, group=Interior Lighting, parameterName=TileDepth X, sortPriority=101}
    B=ScalarParameter() {defaultValue=1, group=Interior Lighting, parameterName=TileDepth Y, sortPriority=101}
   )
   Randomize Rotation
  ) {materialFunction=/Engine/Functions/Engine_MaterialFunctions01/Cubemaps/InteriorCubemap}
  n9 = Divide(
   A=Add(A=ComponentMask(None=@n8.UVW) {r=true}, B=Constant() {r=0.98})
   B=ScalarParameter() {defaultValue=3, group=Interior Lighting, parameterName=Depth Edge Blur, sortPriority=101}
  )
  n10 = Power(Base=LinearInterpolate(A=ScalarParameter() {group=Interior Lighting, parameterName=Depth Min, sortPriority=101}, Alpha=Saturate(None=@n9))) {constExponent=4}
  n11 = NamedRerouteUsage()
  n12 = VectorParameter() {group=Interior Lighting, parameterName=DepthMaskLocation, sortPriority=101, defaultValue=<rgba 0 0 0 1>}
  n13 = ComponentMask(None=@n12.RGB) {g=true, r=true}
  n14 = ScalarParameter() {defaultValue=0.25, group=Interior Lighting, parameterName=DepthMaskSphereRadius, sortPriority=101}
  n15 = ScalarParameter() {group=Interior Lighting, parameterName=DepthMaskSphereHardnes, sortPriority=101}
  n16 = SphereMask(A=@n7, B=@n13, Radius=@n14, Hardness=@n15)
  n17 = ScalarParameter() {defaultValue=2, group=Interior Lighting, parameterName=DepthMaskSpherepower, sortPriority=101}
  n18 = Multiply(A=@n11, B=Saturate(None=Power(Base=@n16, Exp=@n17)))
  n19 = SphereMask(
   A=BumpOffset(Coordinate=@n7, Height=ScalarParameter() {defaultValue=-0.35, group=Interior Lighting, parameterName=DepthMaskBumpOffset, sortPriority=101}) {heightRatio=0.5}
   B=@n13
   Radius=@n14
   Hardness=@n15
  )
  n20 = Multiply(A=@n11, B=Saturate(None=Power(Base=@n19, Exp=@n17)))
  n21 = QualitySwitch(Default=Saturate(None=Multiply(A=@n10, B=@n18)), Low=@n20, Medium=@n20)
  n22 = ScalarParameter() {group=Interior Lighting, parameterName=EmissiveFresnel_Min, sortPriority=101}
  n23 = ScalarParameter() {defaultValue=1.5, group=Interior Lighting, parameterName=EmissiveFresnel_Max, sortPriority=101}
  n24 = DotProduct(A=CameraVectorWS(), B=VertexNormalWS())
  n25 = Multiply(A=@n21, B=Saturate(None=LinearInterpolate(A=@n22, B=@n23, Alpha=@n24)))
  n26 = StaticSwitchParameter(True=LinearInterpolate(A=NamedRerouteUsage(), B=NamedRerouteUsage(), Alpha=@n25), False=NamedRerouteUsage()) {
   desc="Uses SFX Functions texture. Go to  0_SFX_Global group for Flicker Texture input. "
   group=Interior Lighting
   parameterName=Emissive Depth
   sortPriority=-2
  }
  n27 = VectorParameter() {group=Interior Lighting, parameterName=InteriorColorBlendTOD, defaultValue=<rgba 0 0 0 0.6>}
  n28 = ScalarParameter() {defaultValue=1, group=Interior Lighting, parameterName=FakeLighting.SideA.Blend, sortPriority=40}
  n29 = ScalarParameter() {group=Interior Lighting, parameterName=FakeLighting.SideB.Blend, sortPriority=40}
  n30 = FunctionInput(Preview=TwoSidedSign()) {bUsePreviewValueAsDefault=true, inputName=TwoSidedMask, inputType=FunctionInput_Scalar, sortPriority=1}
  n31 = ScalarParameter() {bUseCustomPrimitiveData=true, group=Interior Lighting, parameterName=SwapABFaceSide, primitiveDataIndex=2, sortPriority=38}
  n32 = Saturate(None=@n31)
  n33 = Sign(None=LinearInterpolate(A=@n30, B=Multiply(A=@n30) {constB=-1}, Alpha=Round(None=@n32)))
  n34 = FunctionOutput(None=LinearInterpolate(A=Saturate(None=@n28), B=Saturate(None=@n29), Alpha=Saturate(None=@n33))) {outputName=InteriorDarkening2Sided, sortPriority=5}
  n35 = LinearInterpolate(
   A=NamedRerouteUsage()
   B=LinearInterpolate(
    A=NamedRerouteUsage()
    B=@n26
    Alpha=MaterialFunctionCall(Value_Morning=@n27.R, Value_Day=@n27.R, Value_Evening=@n27.R, Value_Night=@n27.R) {materialFunction=&Scalar_Time_of_Day}.Result
   )
   Alpha=@n34
  )
  n36 = ScalarParameter() {bUseCustomPrimitiveData=true, parameterName=FakeInteriorLight, primitiveDataIndex=1}
  n37 = FunctionOutput(None=Round(None=Saturate(None=@n36))) {description=CPD SwapABFace, outputName=CPD FakeInteriorLight, sortPriority=10}
  n38 = NamedRerouteUsage()
  FunctionOutput(None=LinearInterpolate(A=NamedRerouteUsage(), B=LinearInterpolate(A=@n5, B=@n35, Alpha=@n37), Alpha=@n38)) {outputName=Color}
  n39 = VectorParameter() {
   desc="Blends the Color between the normal glass color with an Interior lighting color to simulate the glass being brighter when lights on the inside of a building are on.\r\n\r\nTurn on Custom Prim Data 11 to enabled interior lighting"
   group=Interior Lighting
   parameterName=FakeInteriorLightingOpacityMultTOD
   channelNames {a=Night, b=Evening, g=Day, r=Morning}
   defaultValue=<rgba 1 1 1 0.374>
  }
  n40 = LinearInterpolate(
   B=MaterialFunctionCall(Value_Morning=@n39.Morning, Value_Day=@n39.Morning, Value_Evening=@n39.Morning, Value_Night=@n39.Morning) {materialFunction=&Scalar_Time_of_Day}.Result
   Alpha=@n34
  ) {constA=1}
  FunctionOutput(None=LinearInterpolate(B=LinearInterpolate(B=@n40, Alpha=@n37) {constA=1}, Alpha=@n38) {constA=1}) {outputName=FakeLightAlphaMulti, sortPriority=1}
  FunctionOutput(None=@n25) {outputName=DepthMask, sortPriority=5}
  n41 = ScalarParameter() {group=A_Emissive_Flicker_V2, parameterName=Use Actor Position?, sortPriority=2}
  n42 = LinearInterpolate(
   A=MaterialFunctionCall() {materialFunction=&ObjectPivotPoint}.Object Pivot Location
   B=ActorPositionWS()
   Alpha=Saturate(None=Round(None=@n41))
  )
  n43 = MaterialFunctionCall(Float3=Frac(None=Multiply(A=@n42) {constB=0.15})) {materialFunction=&BreakOutFloat3Components}
  n44 = Add(A=Add(A=@n43.R, B=@n43.R), B=@n43.B)
  n45 = FunctionInput() {bUsePreviewValueAsDefault=true, inputName=PerPixel Motion Offset, inputType=FunctionInput_Scalar, sortPriority=1}
  n46 = Multiply(
   A=MaterialFunctionCall(
    R=ScalarParameter() {defaultValue=0.628, group=Interior Lighting, parameterName=Flicker Speed 1, sortPriority=1}
    G=ScalarParameter() {defaultValue=1.95, group=Interior Lighting, parameterName=Flicker Speed 2, sortPriority=1}
    B=ScalarParameter() {defaultValue=2.355, group=Interior Lighting, parameterName=Flicker Speed 3, sortPriority=1}
    A=ScalarParameter() {defaultValue=3.123, group=Interior Lighting, parameterName=Flicker Speed 4, sortPriority=1}
   ) {materialFunction=&AppendMany}.RGBA
   B=Add(A=Add(A=@n44, B=@n45), B=Time())
  )
  n47 = Multiply(
   A=Add(A=Frac(None=@n46)) {constB=-0.5}
   B=MaterialFunctionCall(
    R=ScalarParameter() {defaultValue=0.2, group=Interior Lighting, parameterName=Flicker Amplitude 1, sortPriority=1}
    G=ScalarParameter() {defaultValue=0.4, group=Interior Lighting, parameterName=Flicker Amplitude 2, sortPriority=1}
    B=ScalarParameter() {defaultValue=0.2, group=Interior Lighting, parameterName=Flicker Amplitude 3, sortPriority=1}
    A=ScalarParameter() {defaultValue=0.1, group=Interior Lighting, parameterName=Flicker Amplitude 4, sortPriority=1}
   ) {materialFunction=&AppendMany}.RGBA
  ) {constB=4}
  n48 = MaterialFunctionCall(Float4=Multiply(A=Abs(None=@n47)) {constB=2}) {materialFunction=&BreakOutFloat4Components}
  n49 = Add(A=@n48.R, B=@n48.R)
  n50 = Add(A=@n48.B, B=@n48.B)
  n51 = ScalarParameter() {defaultValue=0.5, group=Interior Lighting, parameterName="Minimum Brightness Clamp ( 0-1 Range)", sliderMax=1, sortPriority=2}
  Max(
   A=Add(
    A=Add(A=@n49, B=@n50)
    B=ScalarParameter() {defaultValue=-0.15, group=Interior Lighting, parameterName=Amplitude Range Offset, sortPriority=2}
   )
   B=@n51
  )
  n52 = ScalarParameter() {defaultValue=0.875, group=Interior Lighting, parameterName=Amplitude Min, sortPriority=2}
  n53 = FunctionInput() {bUsePreviewValueAsDefault=true, inputName=PerInstanceRandom, inputType=FunctionInput_Scalar, sortPriority=10, previewValue=<1 1 1 1>}
  n54 = Multiply(
   A=Time()
   B=Multiply(A=@n53, B=ScalarParameter() {defaultValue=0.02, group=Interior Lighting, parameterName=Flicker Speed, sortPriority=101})
  )
  n55 = Multiply(A=@n54) {constB=4.5}
  n56 = Multiply(A=Add(A=Cosine(None=@n55), B=@n54)) {constB=3}
  n57 = Multiply(A=@n54) {constB=4}
  n58 = Add(A=Multiply(A=Sine(None=@n56)) {constB=0.66}, B=Multiply(A=Sine(None=@n57)) {constB=0.5})
  n59 = Add(A=Cosine(None=Multiply(A=@n54) {constB=3.6}), B=Sine(None=@n54))
  n60 = Saturate(None=ConstantBiasScale(None=Multiply(A=@n58, B=@n59)))
  NamedRerouteDeclaration(None=Saturate(None=LinearInterpolate(A=@n52, Alpha=@n60) {constA=0.7})) {name=Flicker, nodeColor=<rgba 0.1719 0 1 1>}
  NamedRerouteDeclaration(None=FunctionInput() {inputName=ColorIn}) {name=InColor, nodeColor=<rgba 1 0.9141 0 1>}
  n61 = Divide(
   A=Distance(A=WorldPosition().XYZ, B=CameraPositionWS())
   B=ScalarParameter() {defaultValue=5000, group=Interior Lighting, parameterName=FakedInteriorColor.FadeRange, sortPriority=47}
  )
  NamedRerouteDeclaration(None=Clamp(
    None=Saturate(None=@n61)
    Min=ScalarParameter() {defaultValue=0.1, group=Interior Lighting, parameterName=FakedInteriorColor.Min, sortPriority=47}
   ) {desc=Clamp to 10% when close, minDefault=0.1}) {name=DistanceFade, nodeColor=<rgba 0.5938 0 1 1>}
  n62 = VectorParameter() {group=Interior Lighting, parameterName=FakedInteriorColor_Morning, defaultValue=<rgba 3 1.95 0.75 0.1451>}
  n63 = VectorParameter() {group=Interior Lighting, parameterName=FakedInteriorColor_Day, defaultValue=<rgba 4 3.2684 0.9043 0.1451>}
  n64 = VectorParameter() {group=Interior Lighting, parameterName=FakedInteriorColor_Evening, defaultValue=<rgba 4 3.2684 0.9043 0.1451>}
  n65 = VectorParameter() {group=Interior Lighting, parameterName=FakedInteriorColor_Night, defaultValue=<rgba 4 3.2684 0.9043 0.1451>}
  NamedRerouteDeclaration(None=MaterialFunctionCall(Color_Morning=@n62.RGB, Color_Day=@n63.RGB, Color_Evening=@n64.RGB, Color_Night=@n65.RGB) {materialFunction=&Color_Time_of_Day}.Result) {name=InteriorLightColor, nodeColor=<rgba 1 0.3516 0 1>}
  n66 = ScalarParameter() {defaultValue=1, group=Interior Lighting, parameterName=EmissiveFresnel_IntMin, sortPriority=101}
  n67 = ScalarParameter() {group=Interior Lighting, parameterName=EmissiveFresnel_IntMax, sortPriority=101}
  n68 = LinearInterpolate(
   A=ScalarParameter() {group=Interior Lighting, parameterName=EmissiveFresnel_IntMinValue, sortPriority=101}
   Alpha=Saturate(None=LinearInterpolate(A=@n66, B=@n67, Alpha=@n24))
  )
  n69 = ScalarParameter() {defaultValue=1, group=Interior Lighting, parameterName=EmissiveFresnel_Contrast, sortPriority=101}
  n70 = PreSkinnedLocalBounds()
  n71 = Subtract(A=VertexInterpolator(VS=PreSkinnedPosition()).PS, B=@n70.Min)
  n72 = Subtract(A=@n70.Max, B=@n70.Max)
  n73 = Add(
   A=ComponentMask(None=Divide(A=@n71, B=@n72)) {b=true}
   B=ScalarParameter() {group=Interior Lighting, parameterName=GradientOffset, sortPriority=101}
  )
  n74 = Power(
   Base=LinearInterpolate(A=ScalarParameter() {group=Interior Lighting, parameterName=Depth Min, sortPriority=101}, Alpha=Saturate(None=@n73))
   Exp=ScalarParameter() {defaultValue=3, group=Interior Lighting, parameterName=GradientContrast, sortPriority=101}
  )
  n75 = StaticSwitchParameter(
   True=LinearInterpolate(A=@n74, B=OneMinus(None=@n74), Alpha=ScalarParameter() {group=Interior Lighting, parameterName=GradientInvert, sortPriority=101})
   False=@n21
  ) {
   desc="Uses SFX Functions texture. Go to  0_SFX_Global group for Flicker Texture input. "
   group=Interior Lighting
   parameterName=Emissive Gradient
   sortPriority=-2
  }
  StaticSwitchParameter(True=Saturate(None=Power(Base=@n68, Exp=@n69)), False=@n75) {
   desc="Uses SFX Functions texture. Go to  0_SFX_Global group for Flicker Texture input. "
   group=Interior Lighting
   parameterName=Emissive Fresnel
   sortPriority=-2
  }
  n76 = Add(A=@n4)
  StaticSwitchParameter(True=ConstantBiasScale(None=Divide(A=@n1, B=@n76)), False=@n6) {desc="Default is UV0 ", group=Interior Lighting, parameterName=Emissive Depth SphericalUV, sortPriority=-30}
  WorldPosition()
 ]}
