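# Material function outputting a window shine mask sampled from a tiling window texture, for fake glints on glass.
# Triplanar projection, in world or pre-skinned local space (for doors and moving objects), scrolls with camera position and can flip its lines.
# The alternative path projects world position along tangent axes for angled windows, adding distance-scaled BumpOffset.
# Optional per-actor variation derives from object position, so it suits only static assets.
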MaterialFunction /WildEstate/Environment/BaseMaterials/MaterialFunctions/MF_WildEstate_WindowShine
graph {nodes [
  n1 = NamedRerouteUsage()
  n2 = Abs(None=ComponentMask(None=@n1) {r=true})
  n3 = Constant() {r=1}
  n4 = Constant()
  n5 = ScalarParameter() {defaultValue=600, group=Window Shine, parameterName=WindowShine.Tiling}
  n6 = Multiply(
   A=Divide(B=Abs(None=@n5)) {constA=256, constB=512}
   B=ScalarParameter() {defaultValue=0.1, group=Window Shine, parameterName=WindowShine.PanSpeed}
  )
  n7 = ObjectPositionWS()
  n8 = Frac(None=Length(None=ComponentMask(None=@n7) {r=true}) {is=class MaterialExpressionLength})
  n9 = StaticSwitchParameter(True=Multiply(A=@n6, B=LinearInterpolate(Alpha=@n8) {constA=0.9, constB=1.1}), False=@n6) {desc=NOTE: Make sure to keep disabled on moving assets/doors, group=Window Shine, parameterName=UsePerActorVariation}
  n10 = MaterialFunctionCall(Float3=Divide(A=CameraPositionWS(), B=Divide(B=@n9) {constA=2048, constB=2048}) {constB=2048}) {materialFunction=&BreakOutFloat3Components}
  n11 = AppendVector(A=Constant(), B=@n10.B)
  n12 = Abs(None=ScalarParameter() {defaultValue=600, group=Window Shine, parameterName=WindowShine.Tiling})
  n13 = Length(None=ObjectPositionWS())
  n14 = Add(A=NamedRerouteUsage(), B=@n13)
  n15 = ScalarParameter() {defaultValue=600, group=Window Shine, parameterName=WindowShine.Tiling}
  n16 = Frac(None=@n13)
  n17 = LinearInterpolate(Alpha=@n16) {constA=0.9, constB=1.3}
  n18 = Multiply(A=Abs(None=Multiply(A=@n15, B=@n17))) {constB=-1}
  n19 = Constant() {r=1000}
  n20 = StaticSwitchParameter(True=Add(A=Divide(A=@n14, B=@n18), B=Multiply(A=@n16, B=@n19)), False=Divide(A=NamedRerouteUsage(), B=Multiply(A=@n12) {constB=-1})) {desc=NOTE: Make sure to keep disabled on moving assets/doors, group=Window Shine, parameterName=UsePerActorVariation}
  n21 = ScalarParameter() {group=Window Shine, parameterName=WindowShine.FlipLines, sliderMax=1}
  n22 = LinearInterpolate(B=Constant2Vector() {g=1, r=-1}, Alpha=Round(None=Saturate(None=@n21))) {constA=1}
  n23 = Add(A=Multiply(A=ComponentMask(None=@n20) {b=true, r=true}, B=@n22), B=Add(A=AppendVector(A=@n10.R, B=@n4), B=@n11))
  n24 = NamedRerouteUsage()
  n25 = Add(A=Multiply(A=ComponentMask(None=@n20) {b=true, g=true}, B=@n22), B=Add(A=AppendVector(A=@n10.G, B=@n4), B=@n11))
  n26 = ComponentMask(None=@n1) {b=true}
  n27 = Multiply(A=ComponentMask(None=@n20) {g=true, r=true}, B=@n22)
  n28 = LinearInterpolate(
   A=LinearInterpolate(
    A=TextureSample(UVs=@n23, Tex=@n24) {desc=y, samplerSource=SSM_Wrap_WorldGroupSettings}.R
    B=TextureSample(UVs=@n25, Tex=@n24) {desc=x, samplerSource=SSM_Wrap_WorldGroupSettings}.R
    Alpha=MaterialFunctionCall(In=@n2, Contrast=@n3) {materialFunction=&CheapContrast}.Result
   ) {desc=Float3}
   B=TextureSample(UVs=@n27, Tex=@n24) {samplerSource=SSM_Wrap_WorldGroupSettings}.R
   Alpha=MaterialFunctionCall(In=Abs(None=@n26), Contrast=@n3) {materialFunction=&CheapContrast}.Result
  )
  n29 = NamedRerouteUsage()
  n30 = Subtract(A=Distance(A=CameraPositionWS(), B=ObjectPositionWS())) {constB=250}
  n31 = LinearInterpolate(B=Constant() {r=2}, Alpha=Saturate(None=Divide(A=@n30) {constB=750})) {constB=0}
  n32 = StaticSwitchParameter(
   True=@n28
   False=TextureSample(UVs=BumpOffset(Coordinate=@n29, Height=@n31), Tex=NamedRerouteUsage()) {samplerSource=SSM_Wrap_WorldGroupSettings}.R
  ) {
   desc="Best for moving objects or objects that rotate (Like Doors)\r\n\r\nFalse - Uses a  \"ROTATED\" World Space, meaning normal world space projection is used, but the space ir rotated based of the normal direction\r\nie windows that are at odd angles (ie 35 degree) will still have correct projection"
   parameterName=UsePreSkinnedLocal
  }
  FunctionOutput(None=StaticSwitchParameter(True=@n28, False=@n32) {desc="Best for continuous windows/glass that are flat (Most WIndows)", parameterName=UseWorldSpace}) {desc=In World Space, outputName=Mask}
  NamedRerouteDeclaration(None=TextureObject() {texture=/WildEstate/Environment/Textures/T_WildEstate_Windows}) {name=TextureObj, nodeColor=<rgba 1 0.4688 0 1>}
  NamedRerouteDeclaration(None=@n25) {name=UV.X, nodeColor=<rgba 0 1 1 1>}
  NamedRerouteDeclaration(None=@n23) {name=UV.Y, nodeColor=<rgba 1 0.6797 0 1>}
  NamedRerouteDeclaration(None=@n27) {name=UV.Z, nodeColor=<rgba 0.3594 0 1 1>}
  n33 = WorldPosition()
  n34 = VertexTangentWS()
  n35 = DotProduct(A=@n33.XYZ, B=@n34)
  n36 = DotProduct(A=@n33.XYZ, B=CrossProduct(A=VertexNormalWS(), B=@n34))
  n37 = Abs(None=ScalarParameter() {defaultValue=600, group=Window Shine, parameterName=WindowShine.Tiling})
  NamedRerouteDeclaration(None=Divide(A=AppendVector(A=@n35, B=@n36), B=Multiply(A=@n37) {constB=-1})) {name=UV.Rotated, nodeColor=<rgba 1 0.0234 0 1>}
  n38 = StaticSwitchParameter(True=VertexInterpolator(VS=PreSkinnedNormal()).PS, False=VertexNormalWS()) {
   desc="Best for moving objects or objects that rotate (Like Doors)\r\n\r\nFalse - Uses a  \"ROTATED\" World Space, meaning normal world space projection is used, but the space ir rotated based of the normal direction\r\nie windows that are at odd angles (ie 35 degree) will still have correct projection"
   parameterName=UsePreSkinnedLocal
  }
  NamedRerouteDeclaration(None=@n38) {name=Normal, nodeColor=<rgba 0.1484 1 0 1>}
  n39 = StaticSwitchParameter(
   True=VertexInterpolator(VS=LocalPosition() {includedOffsets=ExcludeOffsets, localOrigin=InstancePreSkinning}.XYZ).PS
   False=WorldPosition().XYZ
  ) {
   desc="Best for moving objects or objects that rotate (Like Doors)\r\n\r\nFalse - Uses a  \"ROTATED\" World Space, meaning normal world space projection is used, but the space ir rotated based of the normal direction\r\nie windows that are at odd angles (ie 35 degree) will still have correct projection"
   parameterName=UsePreSkinnedLocal
  }
  NamedRerouteDeclaration(None=@n39) {name=Pos, nodeColor=<rgba 1 0 0.4687 1>}
 ]}
